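# Niagara character effect for the Omega emote, layering aura sprites, rising droplets, splatters and a light on the player.
# Aura sprites follow head, arm and hand bones by spawn group; Droplets and Splatter spawn around sampled skeletal bones.
# Curves over normalized loop age synchronize spawn dropouts with red-pink to cyan color shifts mid-loop and near the end, though Droplets only shift at the end.
# Splatter adds two timed bursts on top of its continuous spawn.
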NiagaraSystem /BRCosmetics/Animation/Game/MainPlayer/Emotes/Omega/FX/NS_Emote_Omega_CharacterFX
dependencies {
 usedDataInterfaces [
  &NiagaraDataInterfaceCurve
  &NiagaraDataInterfaceArrayFloat
  &NiagaraDataInterfaceSkeletalMesh
  &NiagaraDataInterfaceArrayFloat3
  &NiagaraDataInterfaceColorCurve
  &NiagaraDataInterfaceArrayColor
  &NiagaraDataInterfaceArrayInt32
  &NiagaraDataInterfaceVector2DCurve
 ]
 usedDynamicInputs []
 usedModules [
  &self:SystemSpawnScript.SystemScriptSource.SystemScriptGraph.NiagaraNodeAssignment_0.SetVariables_#1
  &SystemState
 ]
}
emitters {
 Aura {
  emitterUpdateScript [
   &EmitterState()
   &SpawnRate(
    SpawnRate=&FloatFromCurve(
     FloatCurve=DataInterface() {propertyValues {Curve {keys [
         <at 0 1 RCIM_Constant>
         <at 0.37 1>
         <at 0.3714 0>
         <at 0.4443 0>
         <at 0.45 1>
         {arriveTangent=-2.096, interpMode=RCIM_Cubic, leaveTangent=-2.096, time=0.92, value=1}
         <at 0.9271 0>
         <at 1 0 RCIM_Constant>
        ]}}}
     CurveIndex=$Emitter.NormalizedLoopAge
     Scale Curve=13
    )
    Use Spawn Probability=false
   )
   &SpawnBurst_Instantaneous(Spawn Count=13, Use Spawn Probability=false, UseLoopCountLimit=false)
   /SetVariables(Emitter.TimeScale:NiagaraFloat=&FloatFromCurve(
     FloatCurve=DataInterface() {propertyValues {Curve {keys [
         <at 0 1 RCIM_Constant>
         <at 0.37 1>
         <at 0.3714 0.3>
         <at 0.4443 0.3>
         <at 0.45 1>
         {arriveTangent=-1.4672, interpMode=RCIM_Cubic, leaveTangent=-1.4672, time=0.92, value=1}
         <at 0.9271 0.3>
         <at 1 0.3 RCIM_Constant>
        ]}}}
     CurveIndex=$Emitter.NormalizedLoopAge
    )) {isSetParameters=true}
  ]
  particleSpawnScript [
   &FN_SetSpawnGroup(Number of Spawn Groups=7)
   &InitializeParticle(
    Color=<rgba 1 1 1 0.6>
    Lifetime Max=1.2
    Lifetime Min=0.8
    Lifetime Mode=::NewEnumerator1
    Sprite Size Mode=::NewEnumerator3
    Sprite UV Mode=::NewEnumerator1
    Uniform Sprite Size=&SelectFloatFromArray(Direct Array Index=$Particles.SpawnGroup, Float Selection Array=DataInterface() {propertyValues {FloatData [70,50,50,50,50,25,25]}})
    UsePositionOffset=false
    Write Material Random=true
   )
   &CameraOffset(Camera Offset Amount=-10)
  ]
  particleUpdateScript [
   &ParticleState()
   &ScaleSpriteSize(
    Scale Factor=&Multiply_Vector2D_ByFloat(Float=2, Vector2D=&RandomRangeVector2D(Minimum=<0.7 1>, Maximum=<0.7 1>))
    Scale Sprite Size Mode=::NewEnumerator2
   ) {module=ScaleSpriteSize001}
   &SkeletalMeshLocation(
    Bone / Socket Index=$Particles.SpawnGroup
    Bone Sampling Mode=::NewEnumerator4
    MaskSpawnGroup=false
    Sampled Position Offset=&SelectVectorFromArray(
     Direct Array Index=$Particles.SpawnGroup
     Vector Selection Array=DataInterface() {propertyValues {FloatData [<0 0 0>, <0 0 0>, <0 0 0>, <0 0 0>, <0 0 0>, <5 0 0>, <-5 0 0>, <0 0 0>, <0 0 0>, <0 0 0>, <0 0 0>, <0 0 0>, <0 0 0>]}}
    )
    Skeletal Mesh=DataInterface() {propertyValues {
      DefaultMesh=null
      ExcludeBoneName=root
      OptionalComponentClass=null
      PreviewMesh=/Game/Characters/Player/Male/Male_Avg_Base/Fortnite_M_Avg_Player
      SkinningMode=SkinOnTheFly
      SoftSourceActor=null
      SourceMode=Default
      UvSetIndex=0
      WholeMeshLOD=-1
      bExcludeBone=false
      bReadDeformedGeometry=true
      bRequireCurrentFrameData=true
      ComponentTags []
      FilteredBones [head,upperarm_l,upperarm_r,lowerarm_l,lowerarm_r,hand_l,hand_r]
      FilteredSockets []
      MeshUserParameter {parameter {name=None, typeDefHandle {registeredTypeIndex=1}, varData []}}
      SamplingRegions []
     }}
   )
   &ScaleColor(
    Linear Color Curve=DataInterface() {propertyValues {
      AlphaCurve {keys [<at 0 0 RCIM_Cubic>,<at 0.5 1 RCIM_Cubic>,<at 1 0 RCIM_Cubic>]}
      BlueCurve {keys [<at 0 1>,<at 1 1>]}
      GreenCurve {keys [<at 0 1>,<at 1 1>]}
      RedCurve {keys [<at 0 1>,<at 1 1>]}
     }}
    Scale Mode=::NewEnumerator2
   )
   &ScaleColor(
    Scale RGB=&Lerp_Vector(
     A=&MakeVectorFromLinearColorRGB(Color=&SelectLinearColorFromArray(
       Array Sampling Mode=::NewEnumerator0
       Direct Array Index=?
       Color Selection Array=DataInterface() {propertyValues {ColorData [<rgba 1 0.1978 0.1198 1>, <rgba 1 0.1875 0.2644 1>, <rgba 1 0.0271 0.1103 1>]}}
       Random Seed=0
      ))
     B=&MakeVectorFromLinearColorRGB(Color=<rgba 0 1 0.9529 1>)
     Alpha=&FloatFromCurve(
      FloatCurve=DataInterface() {propertyValues {Curve {keys [<at 0 0 RCIM_Constant>,<at 0.37 0>,<at 0.3714 1>,<at 0.4443 1>,<at 0.45 0>,<at 0.92 0>,<at 0.9271 1>,<at 1 1 RCIM_Constant>]}}}
      CurveIndex=$Emitter.NormalizedLoopAge
     )
    )
    ScaleA=true
    ScaleRGB=true
   ) {module=ScaleColor001}
  ]
  rendererClasses [&NiagaraSpriteRendererProperties]
  renderers [{rendererClass=&NiagaraSpriteRendererProperties}]
 }
 Droplets {
  emitterUpdateScript [
   &EmitterState()
   &SpawnRate(
    SpawnRate=&FloatFromCurve(
     FloatCurve=DataInterface() {propertyValues {Curve {keys [
         <at 0 1 RCIM_Constant>
         <at 0.37 1>
         <at 0.3714 0>
         <at 0.4443 0>
         <at 0.45 1>
         {arriveTangent=-2.096, interpMode=RCIM_Cubic, leaveTangent=-2.096, time=0.92, value=1}
         <at 0.9271 0>
         <at 1 0 RCIM_Constant>
        ]}}}
     CurveIndex=$Emitter.NormalizedLoopAge
     Scale Curve=24
    )
    Use Spawn Probability=false
   )
  ]
  particleSpawnScript [
   &InitializeParticle(
    Lifetime Max=0.7
    Lifetime Min=0.5
    Lifetime Mode=::NewEnumerator1
    Mesh Scale Mode=::NewEnumerator3
    Position Offset=<0 0 75>
    Sprite Rotation Mode=::NewEnumerator2
    Sprite Size Mode=::NewEnumerator1
    Uniform Sprite Size Max=4
    Uniform Sprite Size Min=2
    UsePositionOffset=true
    Write Material Random=true
   ) {module=InitializeParticle001}
   &SkeletalMeshLocation(
    Bone / Socket Index=&SelectIntFromArray(
     Array Sampling Mode=::NewEnumerator1
     Direct Array Index=?
     Int Selection Array=DataInterface() {propertyValues {IntData [0,1]}}
     Random Seed=0
    )
    Bone Sampling Mode=::NewEnumerator4
    MaskSpawnGroup=false
    Skeletal Mesh=$System.Skeletal Mesh
   )
   &ShapeLocation(
    Box Size=<30 15 5>
    MaskSpawnGroup=false
    Offset=<0 0 -5>
    Offset Mode=::NewEnumerator0
    Override Seed=false
    Shape Primitive=::NewEnumerator2
    Surface Only=false
    UseOwnerScale=true
   )
   /Niagara/Modules/Update/SubUV/SubUVAnimation(SubUV Animation Mode=::NewEnumerator1)
   &AddVelocity(Velocity=&RandomRangeVector(Minimum=<0 0 100>, Maximum=<0 1 140>)) {module=AddVelocity001}
   &CameraOffset(Camera Offset Amount=-25)
  ]
  particleUpdateScript [
   &ParticleState()
   &Drag(Drag=&RandomRangeFloat(Minimum=1.5, Maximum=1.2), Use Linear Drag=true, Use Rotational Drag=false)
   &SolveForcesAndVelocity(Clamp Velocity=false, Limit Acceleration=false)
   &ScaleSpriteSize(Scale Factor=&Vector2DFromCurve(
     Vector2Curve=DataInterface() {propertyValues {
       XCurve {keys [<at 0 1>,<at 0.5 1>,<at 0.7 1>,<at 1 0>]}
       YCurve {keys [
         <at 0 3 RCIM_Cubic>
         {arriveTangent=0.2942, interpMode=RCIM_Cubic, leaveTangent=0.2942, tangentMode=RCTM_User, time=0.5, value=1}
         {arriveTangent=0.058, interpMode=RCIM_Cubic, leaveTangent=0.058, tangentMode=RCTM_User, time=0.7, value=1}
         <at 1 0 RCIM_Cubic>
        ]}
      }}
     ScaleCurve=<0.6 1>
    ))
   &ScaleColor(
    Linear Color Curve=DataInterface() {propertyValues {
      AlphaCurve {keys [<at 0 0>,<at 0.15 1>,<at 0.5 1>,<at 1 0>]}
      BlueCurve {keys [<at 0 5>,<at 1 0.2>]}
      GreenCurve {keys [<at 0 5>,<at 1 0.2>]}
      RedCurve {keys [<at 0 5>,<at 1 0.2>]}
     }}
    Scale Mode=::NewEnumerator2
   ) {module=ScaleColor001}
   &ScaleColor(
    Scale RGB=&Lerp_Vector(
     A=&MakeVectorFromLinearColorRGB(Color=&SelectLinearColorFromArray(
       Array Sampling Mode=::NewEnumerator0
       Direct Array Index=?
       Color Selection Array=DataInterface() {propertyValues {ColorData [<rgba 1 0.1313 0.0469 1>, <rgba 1 0.0625 0.1513 1>, <rgba 1 0.0781 0.4289 1>]}}
       Random Seed=0
      ))
     B=&MakeVectorFromLinearColorRGB(Color=<rgba 0 1 0.9529 1>)
     Alpha=&FloatFromCurve(
      FloatCurve=DataInterface() {propertyValues {Curve {keys [<at 0 0 RCIM_Constant>,<at 0.95 1 RCIM_Constant>,<at 1 1 RCIM_Constant>]}}}
      CurveIndex=$Emitter.NormalizedLoopAge
     )
    )
    ScaleA=true
    ScaleRGB=true
   ) {module=ScaleColor002}
  ]
  rendererClasses [&NiagaraSpriteRendererProperties]
  renderers [{rendererClass=&NiagaraSpriteRendererProperties}]
 }
 Light {
  emitterUpdateScript [
   &EmitterState()
   &SpawnBurst_Instantaneous(Use Spawn Probability=false, UseLoopCountLimit=false)
  ]
  particleSpawnScript [&InitializeParticle(Color Mode=::NewEnumerator0, Lifetime=20.57, Position Offset=<0 0 50>, UsePositionOffset=true, Write Material Random=true)]
  particleUpdateScript [
   &ParticleState()
   &ScaleColor(
    Scale RGB=&Lerp_Vector(
     A=&MakeVectorFromLinearColorRGB(Color=&SelectLinearColorFromArray(
       Array Sampling Mode=::NewEnumerator0
       Direct Array Index=?
       Color Selection Array=DataInterface() {propertyValues {ColorData [<rgba 1 0.1978 0.1198 1>, <rgba 1 0.1875 0.2644 1>, <rgba 1 0.0271 0.1103 1>]}}
       Random Seed=0
      ))
     B=&MakeVectorFromLinearColorRGB(Color=<rgba 0 1 0.9529 1>)
     Alpha=&FloatFromCurve(
      FloatCurve=DataInterface() {propertyValues {Curve {keys [<at 0 0 RCIM_Constant>,<at 0.94 0>,<at 0.95 1>,<at 1 1 RCIM_Constant>]}}}
      CurveIndex=$Emitter.NormalizedLoopAge
     )
    )
    ScaleA=true
    ScaleRGB=true
   ) {enabled=false, module=ScaleColor001}
   &FN_Light_Attributes(
    Color=<rgba 3 3 3 1>
    Create Color Attribute=true
    Create Exponent Attribute=true
    Create Position Attribute=true
    Create Radius Attribute=true
    Exponent=4
    Position=<0 30 100>
    Radius=85
    Scale Color=&MakeFloatFromLinearColor(
     LinearColor=/Niagara/DynamicInputs/Lerp/SmoothLerpOverTime_Color(
      Smooth Value:LinearColor=<rgba 1 0.0625 0.1513 1>
      Target Value:LinearColor=<rgba 0 1 0.9529 1>
      Convergence Rate:NiagaraFloat=&FloatFromCurve(
       FloatCurve=DataInterface() {propertyValues {Curve {keys [<at 0 0 RCIM_Constant>,<at 0.37 0>,<at 0.3714 1>,<at 0.4443 1>,<at 0.45 0>,<at 0.9 0>,<at 0.95 1 RCIM_Constant>,<at 1 1 RCIM_Constant>]}}}
       CurveIndex=$Emitter.NormalizedLoopAge
      )
     )
     Channel:ENiagaraLinearColor_Channels=::NewEnumerator0
    )
   )
  ]
  rendererClasses [&NiagaraLightRendererProperties]
  renderers [{rendererClass=&NiagaraLightRendererProperties, rendererIndex=0}]
 }
 Splatter {
  emitterUpdateScript [
   &EmitterState()
   &SpawnBurst_Instantaneous(Spawn Count=5, Spawn Group=1, Spawn Time=7.87, Use Spawn Probability=false, UseLoopCountLimit=false)
   &SpawnBurst_Instantaneous(Spawn Count=5, Spawn Group=1, Spawn Time=19.3, Use Spawn Probability=false, UseLoopCountLimit=false) {module=SpawnBurst_Instantaneous001}
   &SpawnRate(
    SpawnRate=&FloatFromCurve(
     FloatCurve=DataInterface() {propertyValues {Curve {keys [
         <at 0 1 RCIM_Constant>
         <at 0.37 1>
         <at 0.3714 0>
         <at 0.4443 0>
         <at 0.45 1>
         {arriveTangent=-2.096, interpMode=RCIM_Cubic, leaveTangent=-2.096, time=0.92, value=1}
         <at 0.9271 0>
         <at 1 0 RCIM_Constant>
        ]}}}
     CurveIndex=$Emitter.NormalizedLoopAge
     Scale Curve=8
    )
    Use Spawn Probability=false
   )
  ]
  particleSpawnScript [
   /SetVariables(Particles.VisibilityTag=&MaskIntBySpawnGroup(Masked Int=0, Passthrough Int=1, Spawn Group Mask=0)) {isSetParameters=true}
   &InitializeParticle(
    Lifetime=&MaskFloatBySpawnGroup(
     Masked Float=&RandomRangeFloat(Minimum=0.5, Maximum=0.7)
     Passthrough Float=&RandomRangeFloat(Minimum=1.2, Maximum=1.7)
    )
    Mesh Scale Mode=::NewEnumerator3
    Position Offset=<0 0 75>
    Sprite Rotation Mode=::NewEnumerator2
    Sprite Size Max=<25 50>
    Sprite Size Min=<15 35>
    Sprite Size Mode=::NewEnumerator2
    UsePositionOffset=true
    Write Material Random=true
   ) {module=InitializeParticle001}
   &SkeletalMeshLocation(
    Bone / Socket Index=&SelectIntFromArray(
     Array Sampling Mode=::NewEnumerator1
     Direct Array Index=?
     Int Selection Array=DataInterface() {propertyValues {IntData [0,1]}}
     Random Seed=0
    )
    Bone Sampling Mode=::NewEnumerator4
    MaskSpawnGroup=false
    Skeletal Mesh=$System.Skeletal Mesh
   )
   &ShapeLocation(
    Box Size=<30 15 5>
    MaskSpawnGroup=false
    Offset=<0 0 -5>
    Offset Mode=::NewEnumerator0
    Override Seed=false
    Shape Primitive=::NewEnumerator2
    Surface Only=false
    UseOwnerScale=true
   )
   &SubUVAnimation(SubUV Animation Mode=::NewEnumerator1)
   &AddVelocity(Velocity=&MaskVectorBySpawnGroup(
     Masked Vector=&RandomRangeVector(Minimum=<0 0 75>, Maximum=<0 1 120>)
     Passthrough Vector=&RandomRangeVector(Minimum=<0 0 45>, Maximum=<0 1 65>)
    )) {module=AddVelocity001}
   &CameraOffset(Camera Offset Amount=-20)
  ]
  particleUpdateScript [
   &ParticleState()
   &SpriteRotationRate(Rotation Rate=&RandomRangeFloat(Minimum=-90, Maximum=90))
   &Drag(Drag=&RandomRangeFloat(Minimum=0.8, Maximum=1.2), Use Linear Drag=true, Use Rotational Drag=false)
   &SolveForcesAndVelocity(Clamp Velocity=false, Limit Acceleration=false)
   &ScaleSpriteSize(Scale Factor=&Vector2DFromCurve(
     Vector2Curve=DataInterface() {propertyValues {
       XCurve {keys [<at 0 1>,<at 1 1>]}
       YCurve {keys [<at 0 1 RCIM_Cubic>, {arriveTangent=-3.2254, interpMode=RCIM_Cubic, leaveTangent=-3.2254, tangentMode=RCTM_User, time=1, value=0}]}
      }}
     ScaleCurve=<1 1>
    ))
   &ScaleColor(
    Linear Color Curve=DataInterface() {propertyValues {
      AlphaCurve {keys [<at 0 0>,<at 0.15 1>,<at 0.5 1>,<at 1 0>]}
      BlueCurve {keys [<at 0 5>,<at 1 0.2>]}
      GreenCurve {keys [<at 0 5>,<at 1 0.2>]}
      RedCurve {keys [<at 0 5>,<at 1 0.2>]}
     }}
    Scale Mode=::NewEnumerator2
   ) {module=ScaleColor001}
   &ScaleColor(
    Scale RGB=&Lerp_Vector(
     A=&MakeVectorFromLinearColorRGB(Color=&SelectLinearColorFromArray(
       Array Sampling Mode=::NewEnumerator0
       Direct Array Index=?
       Color Selection Array=DataInterface() {propertyValues {ColorData [<rgba 1 0.1313 0.0469 1>, <rgba 1 0.0625 0.1513 1>, <rgba 1 0.0781 0.4289 1>]}}
       Random Seed=0
      ))
     B=&MakeVectorFromLinearColorRGB(Color=<rgba 0 1 0.9529 1>)
     Alpha=&FloatFromCurve(
      FloatCurve=DataInterface() {propertyValues {Curve {keys [<at 0 0 RCIM_Constant>,<at 0.37 0>,<at 0.3714 1>,<at 0.4443 1>,<at 0.45 0>,<at 0.9 0>,<at 0.95 1 RCIM_Constant>,<at 1 1 RCIM_Constant>]}}}
      CurveIndex=$Emitter.NormalizedLoopAge
     )
    )
    ScaleA=true
    ScaleRGB=true
   ) {module=ScaleColor002}
   &DynamicMaterialParameters(
    Index 0 Param 1=&FloatFromCurve(FloatCurve=DataInterface() {propertyValues {Curve {keys [
         {arriveTangent=0.0674, interpMode=RCIM_Cubic, leaveTangent=0.0674, tangentMode=RCTM_User, time=0, value=1}
         {arriveTangent=-2.8897, interpMode=RCIM_Cubic, leaveTangent=-2.8897, tangentMode=RCTM_User, time=1, value=0}
        ]}}})
    Index 2 Param 4=1.2
    Module.Param0WriteEnabled=true
    Module.Param0WriteEnabled2=true
    Module.Param1WriteEnabled=false
    Module.Param1WriteEnabled2=true
    Module.Param2WriteEnabled=false
    Module.Param2WriteEnabled2=true
    Module.Param3WriteEnabled=false
    Module.Param3WriteEnabled2=true
    Write Parameter Index 2=true
   )
  ]
  rendererClasses [&NiagaraSpriteRendererProperties]
  renderers [
   {rendererClass=&NiagaraSpriteRendererProperties}
   {rendererClass=&NiagaraSpriteRendererProperties, rendererIndex=1}
  ]
 }
}
summary {
 systemName=NS_Emote_Omega_CharacterFX
 emitters [
  {emitterName=Aura, rendererClasses [&NiagaraSpriteRendererProperties]}
  {emitterName=Splatter, rendererClasses [&NiagaraSpriteRendererProperties]}
  {emitterName=Droplets, rendererClasses [&NiagaraSpriteRendererProperties]}
  {emitterName=Light, rendererClasses [&NiagaraLightRendererProperties]}
 ]
 userVariables [
  User.bIsFrontEnd {type=NiagaraBool}
  User.bIsFrontEndPreview {type=NiagaraBool}
 ]
}
