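# Niagara burst effect for the Omega emote: a short multicolor shockwave, glow and spark pop.
# Color_wave splits two spawn groups by visibility tag, EnergyWave_Burst and RayGlow add expanding meshes, and Sparks scatters curl-noise sprites.
# Repeated color-over-life curves shift hues, while material parameters vary by platform and front-end context and wave alpha by time of day.
# GlitchDecal and EnergyWave_Burst's light-attribute module are disabled.
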NiagaraSystem /BRCosmetics/Animation/Game/MainPlayer/Emotes/Omega/FX/NS_Emote_Omega_ColorBurst
dependencies {
 usedDataInterfaces [&NiagaraDataInterfaceCurve, &NiagaraDataInterfaceColorCurve, &NiagaraDataInterfacePlatformSet]
 usedDynamicInputs []
 usedModules [&SystemState]
}
emitters {
 Color_wave {
  emitterUpdateScript [
   &EmitterState()
   &SpawnBurst_Instantaneous(Use Spawn Probability=false, UseLoopCountLimit=false)
   &SpawnBurst_Instantaneous(Spawn Group=1, Use Spawn Probability=false, UseLoopCountLimit=false) {module=SpawnBurst_Instantaneous001}
  ]
  particleSpawnScript [
   &InitializeParticle(
    Color=<rgba 3 3 3 2>
    Lifetime=&MaskFloatBySpawnGroup(Masked Float=0.5, Passthrough Float=0.6)
    Mesh Scale Mode=::NewEnumerator3
    Mesh Uniform Scale=0.9
    Position Offset=<0 25 0>
    Sprite Rotation Mode=::NewEnumerator2
    Sprite Size Mode=::NewEnumerator3
    Uniform Sprite Size=65
    UsePositionOffset=true
    Write Material Random=true
   )
   /SetVariables(Particles.VisibilityTag=&MaskIntBySpawnGroup(Masked Int=0, Passthrough Int=1, Spawn Group Mask=0)) {isSetParameters=true}
   &InitialMeshOrientation(
    Orientation Vector=<0 0 1>
    Rotation=&RandomRangeVector(Minimum=<0 0 0>, Maximum=<0 0 1>)
    Use Orientation Vector=true
    Use Rotation Vector=true
   )
   /SetVariables(Particles.ShockwaveCO=10) {isSetParameters=true}
  ]
  particleUpdateScript [
   &ParticleState() {module=ParticleState001}
   &ScaleSpriteSize(Uniform Curve Sprite Scale=DataInterface() {propertyValues {Curve {keys [<at 0 0>,<at 0.1 1.4>,<at 1 2>]}}}) {module=ScaleSpriteSize001}
   &ScaleSpriteSize(
    Uniform Curve Scale=1.2
    Uniform Curve Sprite Scale=DataInterface() {propertyValues {Curve {keys [
        {arriveTangent=6.5767, interpMode=RCIM_Cubic, leaveTangent=6.5767, tangentMode=RCTM_User, time=0, value=0}
        {arriveTangent=0.7192, interpMode=RCIM_Cubic, leaveTangent=0.7192, tangentMode=RCTM_User, time=0.2305, value=0.7836}
        {tangentMode=RCTM_User, time=1, value=1}
       ]}}}
   )
   &ScaleMeshSize(Scale Factor=&VectorFromFloat(Value=&FloatFromCurve(
      FloatCurve=DataInterface() {propertyValues {Curve {keys [<at 0 0>,<at 0.1 1.4>,<at 1 2>]}}}
      Scale Curve=&Multiply_Float(A=1, B=1)
     )))
   &DynamicMaterialParameters(
    Index 0 Param 1=1
    Index 0 Param 3=1
    Index 0 Param 4=1
    Module.Param0WriteEnabled=true
    Module.Param1WriteEnabled=true
    Module.Param2WriteEnabled=true
    Module.Param3WriteEnabled=true
   )
   /SetVariables(
    Particles.ShockwaveColor:LinearColor=&ColorFromCurve(ColorCurve=DataInterface() {propertyValues {
       AlphaCurve {keys [<at 0 1>,<at 1 0>]}
       BlueCurve {keys [<at 0.1259 2 RCIM_Cubic>,<at 0.2277 1.8773>,<at 0.4 0>,<at 0.5655 0>,<at 0.75 0.2851 RCIM_Cubic>,<at 1 0>]}
       GreenCurve {keys [
         <at 0.1259 1.1188 RCIM_Cubic>
         <at 0.2277 3>
         <at 0.4 1.9095>
         <at 0.5655 0.9115>
         {arriveTangent=0.2037, interpMode=RCIM_Cubic, leaveTangent=0.2037, time=0.75, value=0}
         <at 1 1>
        ]}
       RedCurve {keys [
         <at 0.1259 0.6458 RCIM_Cubic>
         <at 0.2277 1.2201>
         <at 0.4 2.5926>
         <at 0.5655 1.6785>
         {arriveTangent=-1.6002, interpMode=RCIM_Cubic, leaveTangent=-1.6002, time=0.75, value=2}
         <at 1 0.9833>
        ]}
      }})
    Particles.ShockwaveDynamicMaterialParameter:Vector4f=&MakeVector4(
     X=&FloatFromCurve(FloatCurve=DataInterface() {propertyValues {Curve {keys [{arriveTangent=-2.1015, interpMode=RCIM_Cubic, leaveTangent=-2.1015, tangentMode=RCTM_User, time=0, value=0.5}, <at 1 -0.15>]}}})
     Y=&FloatFromCurve(FloatCurve=DataInterface() {propertyValues {Curve {keys [{arriveTangent=-0.3464, interpMode=RCIM_Cubic, leaveTangent=-0.3464, tangentMode=RCTM_User, time=0, value=1}, <at 1 0.9>]}}})
     Z=&FloatFromCurve(FloatCurve=DataInterface() {propertyValues {Curve {keys [<at 0 0>,<at 0.15 1>,<at 1 1.2>]}}})
     W=0
    )
   ) {isSetParameters=true}
   &ScaleColor(
    Scale Alpha=&FloatFromCurve(
     FloatCurve=DataInterface() {propertyValues {Curve {keys [
         {interpMode=RCIM_Cubic, tangentMode=RCTM_User, time=0, value=0}
         {interpMode=RCIM_Cubic, tangentMode=RCTM_User, time=0.2, value=1}
         {tangentMode=RCTM_User, time=1, value=0}
        ]}}}
     Scale Curve=&Multiply_Float(A=2, B=&FNTimeOfDayScalar(Night Scale=0.5, Morning Scale=0.8, Evening Scale=0.8))
    )
    Scale RGB=&MakeVectorFromLinearColorRGB(Color=&MultiplyLinearColor(
      Linear Color=&ColorFromCurve(ColorCurve=DataInterface() {propertyValues {
         AlphaCurve {keys []}
         BlueCurve {keys [
           <at 0.1259 2 RCIM_Cubic>
           <at 0.2277 5>
           <at 0.4 2.5926>
           <at 0.5655 0.4735>
           {arriveTangent=-1.0898, interpMode=RCIM_Cubic, leaveTangent=-1.0898, time=0.75, value=0}
           <at 1 0>
          ]}
         GreenCurve {keys [
           <at 0.1259 0.1681 RCIM_Cubic>
           <at 0.2277 1.5012>
           <at 0.4 0>
           <at 0.5655 0.0027>
           {arriveTangent=2.2954, interpMode=RCIM_Cubic, leaveTangent=2.2954, time=0.75, value=0.566}
           <at 1 1>
          ]}
         RedCurve {keys [
           <at 0.1259 0.5589 RCIM_Cubic>
           <at 0.2277 0>
           <at 0.4 1.3901>
           <at 0.5655 1.6785>
           {arriveTangent=-2.5385, interpMode=RCIM_Cubic, leaveTangent=-2.5385, time=0.75, value=2}
           <at 1 0.5756>
          ]}
        }})
      Scale Factor=<1 1 1 1>
     ))
    ScaleA=true
    ScaleRGB=true
   )
  ]
  rendererClasses [&NiagaraSpriteRendererProperties, &NiagaraMeshRendererProperties]
  renderers [
   {rendererClass=&NiagaraSpriteRendererProperties}
   {rendererClass=&NiagaraMeshRendererProperties, rendererIndex=1}
  ]
 }
 EnergyWave_Burst {
  emitterUpdateScript [
   &EmitterState()
   &SpawnBurst_Instantaneous(Use Spawn Probability=false, UseLoopCountLimit=false)
  ]
  particleSpawnScript [&InitializeParticle(Mesh Scale Mode=::NewEnumerator3, Mesh Uniform Scale=1.7, Position Offset=<0 0 -80>, UsePositionOffset=true, Write Material Random=true)]
  particleUpdateScript [
   &ParticleState()
   &ScaleMeshSize(Scale Factor=&VectorFromFloat(Value=&FloatFromCurve(
      FloatCurve=DataInterface() {propertyValues {Curve {keys [<at 0 0 RCIM_Cubic>, {arriveTangent=1, interpMode=RCIM_Cubic, leaveTangent=1, time=0.2041, value=0.6997}, <at 1 1 RCIM_Cubic>]}}}
      Scale Curve=0.9
     )))
   &ScaleColor(
    Linear Color Curve=DataInterface() {propertyValues {
      AlphaCurve {keys [<at 0 1>,<at 1 0>]}
      BlueCurve {keys [<at 0.1259 2 RCIM_Cubic>,<at 0.2277 1.8773>,<at 0.4 0>,<at 0.5655 0>,<at 0.75 0.2851 RCIM_Cubic>,<at 1 0>]}
      GreenCurve {keys [
        <at 0.1259 1.1188 RCIM_Cubic>
        <at 0.2277 3>
        <at 0.4 1.9095>
        <at 0.5655 0.9115>
        {arriveTangent=0.2037, interpMode=RCIM_Cubic, leaveTangent=0.2037, time=0.75, value=0}
        <at 1 1>
       ]}
      RedCurve {keys [
        <at 0.1259 0.6458 RCIM_Cubic>
        <at 0.2277 1.2201>
        <at 0.4 2.5926>
        <at 0.5655 1.6785>
        {arriveTangent=-1.6002, interpMode=RCIM_Cubic, leaveTangent=-1.6002, time=0.75, value=2}
        <at 1 0.9833>
       ]}
     }}
    Scale Mode=::NewEnumerator2
   )
   &DynamicMaterialParameters(
    Index 0 Param 1=&FloatFromCurve(FloatCurve=DataInterface() {propertyValues {Curve {keys [<at 0 1>,<at 0.2 1.375>,<at 1 0>]}}}, Scale Curve=1.2)
    Index 0 Param 2=2
    Index 0 Param 3=&FloatFromCurve(
     FloatCurve=DataInterface() {propertyValues {Curve {keys [<at 0.1 1>,<at 1 0>]}}}
     Scale Curve=&FN_FrontEndFloatModifier(Gameplay=6, Front End=3)
    )
    Index 0 Param 4=&FloatFromCurve(FloatCurve=DataInterface() {propertyValues {Curve {keys [
         <at 0 1>
         {arriveTangent=0.3079, interpMode=RCIM_Cubic, leaveTangent=0.3079, tangentMode=RCTM_User, time=0.3, value=0.9}
         <at 1 1 RCIM_Cubic>
        ]}}})
    Index 1 Param 1=&FN_PickFloatDependOnPlatform(TargetPlatformActive=180, TargetPlatformNoActive=100)
    Index 1 Param 2=&FN_PickFloatDependOnPlatform(TargetPlatformActive=2.5, TargetPlatformNoActive=5)
    Module.Index0Param0Write=true
    Module.Index0Param1Write=true
    Module.Index0Param2Write=true
    Module.Index0Param3Write=true
    Module.Index1Param0Write=true
    Module.Index1Param1Write=true
    Module.Index1Param2Write=true
    Module.Index1Param3Write=true
    Write Parameter Index 1=true
   )
   &FN_Light_Attributes(
    Color=&ColorFromCurve(ColorCurve=DataInterface() {propertyValues {
       AlphaCurve {keys [<at 0 0.7>]}
       BlueCurve {keys [<at 0.1259 2 RCIM_Cubic>,<at 0.2277 1.8773>,<at 0.4 0>,<at 0.5655 0>,<at 0.75 0.2851 RCIM_Cubic>,<at 1 0>]}
       GreenCurve {keys [
         <at 0.1259 1.1188 RCIM_Cubic>
         <at 0.2277 3>
         <at 0.4 1.9095>
         <at 0.5655 0.9115>
         {arriveTangent=0.2037, interpMode=RCIM_Cubic, leaveTangent=0.2037, time=0.75, value=0}
         <at 1 1>
        ]}
       RedCurve {keys [
         <at 0.1259 0.6458 RCIM_Cubic>
         <at 0.2277 1.2201>
         <at 0.4 2.5926>
         <at 0.5655 1.6785>
         {arriveTangent=-1.6002, interpMode=RCIM_Cubic, leaveTangent=-1.6002, time=0.75, value=2}
         <at 1 0.9833>
        ]}
      }})
    Create Color Attribute=true
    Create Exponent Attribute=true
    Create Radius Attribute=true
    Exponent=4
    Radius=375
    Scale Color=&FloatFromCurve(FloatCurve=DataInterface() {propertyValues {Curve {keys [<at 0 1>,<at 1 0>]}}}, Scale Curve=3)
   ) {enabled=false}
  ]
  rendererClasses [&NiagaraMeshRendererProperties, &NiagaraLightRendererProperties]
  renderers [
   {rendererClass=&NiagaraMeshRendererProperties, rendererIndex=0}
   {rendererClass=&NiagaraLightRendererProperties, rendererIndex=1}
  ]
 }
 GlitchDecal {
  bEnabled=false
  emitterUpdateScript [
   &EmitterState()
   &SpawnBurst_Instantaneous(Use Spawn Probability=false, UseLoopCountLimit=false)
  ]
  particleSpawnScript [
   &InitializeParticle(
    Color Maximum=<rgba 5 2.4697 4.6721 1>
    Color Minimum=<rgba 5 3.9485 1.8759 1>
    Color Mode=::NewEnumerator2
    Lifetime=0.8
    Position Offset=<0 0 100>
    Sprite Size Mode=::NewEnumerator3
    Uniform Sprite Size=200
    UsePositionOffset=false
    Write Material Random=true
   )
   &CameraOffset(Camera Offset Amount=30)
  ]
  particleUpdateScript [
   &ParticleState()
   &ScaleColor(
    Scale Alpha=&FloatFromCurve(FloatCurve=DataInterface() {propertyValues {Curve {keys [<at 0 1>,<at 1 0>]}}})
    ScaleA=true
    ScaleRGB=true
   )
  ]
  rendererClasses [&NiagaraSpriteRendererProperties]
  renderers [{rendererClass=&NiagaraSpriteRendererProperties}]
 }
 RayGlow {
  emitterUpdateScript [
   &EmitterState()
   &SpawnBurst_Instantaneous(Spawn Count=1, Use Spawn Probability=false)
  ]
  particleSpawnScript [
   &InitializeParticle(
    Lifetime=0.35
    Mesh Scale Mode=::NewEnumerator3
    Mesh Uniform Scale=0.375
    Position Offset=<0 25 0>
    Sprite Size Mode=::NewEnumerator3
    Uniform Sprite Size=300
    UsePositionOffset=true
    Write Material Random=true
   )
   &InitialMeshOrientation(
    Rotation=&RandomRangeVector(Minimum=<0 0 0>, Maximum=<1 1 1>)
    Use Orientation Vector=true
    Use Rotation Vector=true
   ) {module=InitialMeshOrientation001}
  ]
  particleUpdateScript [
   &ParticleState()
   &ScaleMeshSize(Scale Factor=&VectorFromFloat(Value=&FloatFromCurve(FloatCurve=DataInterface() {propertyValues {Curve {keys [<at 0 0 RCIM_Cubic>, {arriveTangent=1, interpMode=RCIM_Cubic, leaveTangent=1, time=0.298, value=0.687}, <at 1 1 RCIM_Cubic>]}}})))
   &ScaleSpriteSize(Uniform Curve Sprite Scale=DataInterface() {propertyValues {Curve {keys [<at 0 0.2 RCIM_Cubic>, {arriveTangent=-0.2, interpMode=RCIM_Cubic, leaveTangent=-0.2, time=0.6, value=1}, <at 1 0 RCIM_Cubic>]}}})
   &SolveForcesAndVelocity(Clamp Velocity=false, Limit Acceleration=false)
   &ScaleColor(
    Linear Color Curve=DataInterface() {propertyValues {
      AlphaCurve {keys [<at 0 1>,<at 1 0>]}
      BlueCurve {keys [
        <at 0.1259 2 RCIM_Cubic>
        <at 0.2277 5>
        <at 0.4 2.5926>
        <at 0.5655 0.4735>
        {arriveTangent=-1.0898, interpMode=RCIM_Cubic, leaveTangent=-1.0898, time=0.75, value=0}
        <at 1 0>
       ]}
      GreenCurve {keys [
        <at 0.1259 0.1681 RCIM_Cubic>
        <at 0.2277 1.5012>
        <at 0.4 0>
        <at 0.5655 0.0027>
        {arriveTangent=2.2954, interpMode=RCIM_Cubic, leaveTangent=2.2954, time=0.75, value=0.566}
        <at 1 1>
       ]}
      RedCurve {keys [
        <at 0.1259 0.5589 RCIM_Cubic>
        <at 0.2277 0>
        <at 0.4 1.3901>
        <at 0.5655 1.6785>
        {arriveTangent=-2.5385, interpMode=RCIM_Cubic, leaveTangent=-2.5385, time=0.75, value=2}
        <at 1 0.5756>
       ]}
     }}
    Scale Mode=::NewEnumerator2
   )
  ]
  rendererClasses [&NiagaraMeshRendererProperties, &NiagaraSpriteRendererProperties]
  renderers [
   {rendererClass=&NiagaraMeshRendererProperties, rendererIndex=0}
   {rendererClass=&NiagaraSpriteRendererProperties, rendererIndex=1}
  ]
 }
 Sparks {
  emitterUpdateScript [
   &EmitterState()
   &SpawnBurst_Instantaneous(Spawn Count=100, Use Spawn Probability=false, UseLoopCountLimit=false)
  ]
  particleSpawnScript [
   &InitializeParticle(
    Lifetime Max=0.6
    Lifetime Min=0.2
    Lifetime Mode=::NewEnumerator1
    Sprite Rotation Mode=::NewEnumerator2
    Sprite Size Mode=::NewEnumerator1
    Uniform Sprite Size Max=3
    Uniform Sprite Size Min=1
    UsePositionOffset=false
    Write Material Random=true
   )
   &ShapeLocation(MaskSpawnGroup=false, Override Seed=false, Sphere Radius=10, UseOwnerScale=true)
   &AddVelocity(
    Module.Use Falloff Distance=false
    Override Seed=false
    Velocity Mode=::NewEnumerator1
    Velocity Speed=&RandomRangeFloat(Minimum=180, Maximum=350)
   )
  ]
  particleUpdateScript [
   &ParticleState()
   &ScaleSpriteSize(Uniform Curve Sprite Scale=DataInterface() {propertyValues {Curve {keys [<at 0 1 RCIM_Cubic>, {arriveTangent=-1.9212, interpMode=RCIM_Cubic, leaveTangent=-1.9212, tangentMode=RCTM_User, time=1, value=0}]}}})
   &CurlNoiseForce(Bias Noise=false, Noise Strength=&RandomRangeFloat(Minimum=450, Maximum=700), Pan Noise=false)
   &Drag(
    Drag=&FloatFromCurve(
     FloatCurve=DataInterface() {propertyValues {Curve {keys [<at 0 0>,<at 0.1 1 RCIM_Cubic>]}}}
     Scale Curve=&RandomRangeFloat(Minimum=0.8, Maximum=1.2)
    )
    Use Linear Drag=true
    Use Rotational Drag=false
   )
   &ScaleColor(
    Linear Color Curve=DataInterface() {propertyValues {
      AlphaCurve {keys [<at 0 1>,<at 1 0>]}
      BlueCurve {keys [
        <at 0.1259 2 RCIM_Cubic>
        <at 0.2277 5>
        <at 0.4 2.5926>
        <at 0.5655 0.4735>
        {arriveTangent=-1.0898, interpMode=RCIM_Cubic, leaveTangent=-1.0898, time=0.75, value=0}
        <at 1 0>
       ]}
      GreenCurve {keys [
        <at 0.1259 0.1681 RCIM_Cubic>
        <at 0.2277 1.5012>
        <at 0.4 0>
        <at 0.5655 0.0027>
        {arriveTangent=2.2954, interpMode=RCIM_Cubic, leaveTangent=2.2954, time=0.75, value=0.566}
        <at 1 1>
       ]}
      RedCurve {keys [
        <at 0.1259 0.5589 RCIM_Cubic>
        <at 0.2277 0>
        <at 0.4 1.3901>
        <at 0.5655 1.6785>
        {arriveTangent=-2.5385, interpMode=RCIM_Cubic, leaveTangent=-2.5385, time=0.75, value=2}
        <at 1 0.5756>
       ]}
     }}
    Scale Mode=::NewEnumerator2
   )
   &SolveForcesAndVelocity(Clamp Velocity=false, Limit Acceleration=false)
   &ScaleSpriteSizeBySpeed(Max Scale Factor=<0.5 3>, Velocity Threshold=250)
  ]
  rendererClasses [&NiagaraSpriteRendererProperties]
  renderers [{rendererClass=&NiagaraSpriteRendererProperties}]
 }
}
summary {
 systemName=NS_Emote_Omega_ColorBurst
 emitters [
  {emitterName=RayGlow, rendererClasses [&NiagaraMeshRendererProperties, &NiagaraSpriteRendererProperties]}
  {emitterName=Color_wave, rendererClasses [&NiagaraSpriteRendererProperties, &NiagaraMeshRendererProperties]}
  {
   emitterName=EnergyWave_Burst
   rendererClasses [&NiagaraMeshRendererProperties, &NiagaraLightRendererProperties]
  }
  {bEnabled=false, emitterName=GlitchDecal, rendererClasses [&NiagaraSpriteRendererProperties]}
  {emitterName=Sparks, rendererClasses [&NiagaraSpriteRendererProperties]}
 ]
 userVariables [
  User.bIsFrontEnd {type=NiagaraBool}
  User.bIsFrontEndPreview {type=NiagaraBool}
 ]
}
