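# Niagara debris effect for the Omega emote: a burst of cracks, glass shards, streaks, hologram lines and a glowing halo with rays and wave, mostly cyan to magenta.
# Glass spawns on the skeletal mesh and a ring, flies out, slows under drag, then point attraction pulls it back, with delayed inward streaks.
# User positions place cracks and halo.
# Spawn groups split variants within emitters, and hologram spawn rate differs between front end and gameplay.
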NiagaraSystem /BRCosmetics/Animation/Game/MainPlayer/Emotes/Omega/FX/NS_Emote_Omega_Debris
dependencies {
 usedDataInterfaces [
  &NiagaraDataInterfaceArrayFloat
  &NiagaraDataInterfaceMeshRendererInfo
  &NiagaraDataInterfaceCurve
  &NiagaraDataInterfaceColorCurve
  &NiagaraDataInterfaceVector2DCurve
  &NiagaraDataInterfaceSkeletalMesh
 ]
 usedDynamicInputs [&SelectFloatFromArray, &ReturnNormalizedExecIndex]
 usedModules [
  &self:SystemSpawnScript.SystemScriptSource.SystemScriptGraph.NiagaraNodeAssignment_0.SetVariables_#1
  &SystemState
 ]
}
emitters {
 Cracks {
  emitterUpdateScript [
   &EmitterState()
   &SpawnBurst_Instantaneous(Use Spawn Probability=false, UseLoopCountLimit=false)
   &SpawnBurst_Instantaneous(Spawn Group=1, Use Spawn Probability=false, UseLoopCountLimit=false) {module=SpawnBurst_Instantaneous003}
  ]
  particleSpawnScript [
   &FN_SetSpawnGroup(Number of Spawn Groups=2)
   &InitializeParticle(
    Color=<rgba 0.5195 0.9038 1 1>
    Lifetime=1.35
    Mesh Scale=&Multiply_VectorByFloat(Vector=&RandomRangeVector(Minimum=<7 1 3>, Maximum=<4 2.5 0.9>), Float=0.7)
    Mesh Scale Mode=::NewEnumerator4
    Position Offset=$User.Crack Position
    Sprite Size Mode=::NewEnumerator3
    Uniform Sprite Size=150
    UsePositionOffset=true
    Write Material Random=true
   )
   &ShapeLocation(
    Handle Radius=30
    Large Radius=60
    MaskSpawnGroup=false
    Rotation Angle=0.5
    Rotation Axis=<0 0 1>
    Rotation Mode=::NewEnumerator1
    Shape Primitive=::NewEnumerator3
    Torus Distribution Mode:ENiagaraTorusDistributionMode=::NewEnumerator1
    U Position=&MaskFloatBySpawnGroup(
     Masked Float=&RandomRangeFloat(Minimum=0, Maximum=0.2)
     Passthrough Float=&RandomRangeFloat(Minimum=0.4, Maximum=0.6)
    )
    UseOwnerScale=true
    V Position=&MaskFloatBySpawnGroup(
     Masked Float=&RandomRangeFloat(Minimum=0, Maximum=0.2)
     Passthrough Float=&RandomRangeFloat(Minimum=0, Maximum=0.15)
    )
   )
   &InitialMeshOrientation(
    Mesh Orientation Mode=::NewEnumerator2
    Rotation=&RandomRangeVector(Minimum=<0 0 0>, Maximum=<1 1 1>)
    Use Rotation Vector=true
   )
   /SetVariables(Particles.Sprite CameraOffset=-40) {isSetParameters=true}
  ]
  particleUpdateScript [
   &ParticleState()
   &ScaleMeshSize(Scale Factor=&VectorFromFloat(Value=&FloatFromCurve(FloatCurve=DataInterface() {propertyValues {Curve {keys [{arriveTangent=0.803, interpMode=RCIM_Cubic, time=0, value=0.6}, {interpMode=RCIM_Cubic, leaveTangent=0.136, time=1, value=1}]}}})))
   /SetVariables(Particles.SpriteColor=&ColorFromCurve(ColorCurve=DataInterface() {propertyValues {
       AlphaCurve {keys [<at 0 0.2 RCIM_Cubic>,<at 0.5 0.2>,<at 1 0 RCIM_Cubic>]}
       BlueCurve {keys [<at 0 1 RCIM_Cubic>,<at 1 0.3542 RCIM_Cubic>]}
       GreenCurve {keys [<at 0 0.6863 RCIM_Cubic>,<at 1 0 RCIM_Cubic>]}
       RedCurve {keys [<at 0 0 RCIM_Cubic>,<at 1 0.0147 RCIM_Cubic>]}
      }})) {isSetParameters=true}
   &ScaleColor(
    Scale Alpha=&FloatFromCurve(
     FloatCurve=DataInterface() {propertyValues {Curve {keys [<at 0.9 1 RCIM_Cubic>, {arriveTangent=-16.6393, interpMode=RCIM_Cubic, leaveTangent=-16.6393, tangentMode=RCTM_User, time=1, value=0}]}}}
     Scale Curve=0.4
    )
    Scale RGB=&VectorFromFloat(Value=&FloatFromCurve(FloatCurve=DataInterface() {propertyValues {Curve {keys [
          <at 0 1>
          {arriveTangentWeight=0.3005, interpMode=RCIM_Cubic, leaveTangentWeight=0.3005, tangentMode=RCTM_User, time=0.75, value=1}
          {interpMode=RCIM_Cubic, leaveTangent=-3.378, time=1, value=0.2}
         ]}}}))
    ScaleA=true
    ScaleRGB=true
   )
   &DynamicMaterialParameters(
    Index 0 Param 1=&FloatFromCurve(FloatCurve=DataInterface() {propertyValues {Curve {keys [
         {arriveTangent=1.3472, interpMode=RCIM_Cubic, leaveTangent=1.3472, tangentMode=RCTM_User, time=0, value=0}
         {arriveTangent=-0.0333, interpMode=RCIM_Cubic, leaveTangent=-0.0333, tangentMode=RCTM_User, time=0.55, value=1}
         {arriveTangent=-0.0993, interpMode=RCIM_Cubic, leaveTangent=-0.0993, tangentMode=RCTM_User, time=0.9, value=0}
        ]}}})
    Index 0 Param 2=&FloatFromCurve(
     FloatCurve=DataInterface() {propertyValues {Curve {keys [
         {arriveTangent=12.6854, interpMode=RCIM_Cubic, leaveTangent=12.6854, tangentMode=RCTM_User, time=0, value=0}
         {arriveTangentWeight=0.0536, interpMode=RCIM_Cubic, leaveTangentWeight=0.1786, tangentMode=RCTM_User, time=0.4, value=1}
        ]}}}
     Scale Curve=0.25
    )
    Index 0 Param 3=&Add_Float(
     A=&FloatFromCurve(
      FloatCurve=DataInterface() {propertyValues {Curve {keys [{arriveTangent=0.6161, interpMode=RCIM_Cubic, leaveTangent=0.6161, tangentMode=RCTM_User, time=0, value=-0.1}, <at 1 0>]}}}
      Scale Curve=0.8
     )
     B=&RandomRangeFloat(Minimum=0, Maximum=1)
    )
    Module.Index0Param0Write=true
    Module.Index0Param1Write=true
    Module.Index0Param2Write=true
    Module.Index0Param3Write=true
   )
  ]
  rendererClasses [&NiagaraMeshRendererProperties, &NiagaraSpriteRendererProperties]
  renderers [
   {rendererClass=&NiagaraMeshRendererProperties, rendererIndex=0}
   {rendererClass=&NiagaraSpriteRendererProperties, rendererIndex=1}
  ]
 }
 Glass_Body_EQ {
  emitterUpdateScript [
   &EmitterState()
   &SpawnBurst_Instantaneous(
    Spawn Count=&RandomRangeInt(Minimum=5, Maximum=8)
    Use Spawn Probability=false
    UseLoopCountLimit=false
   )
   &SpawnBurst_Instantaneous(
    Spawn Count=&RandomRangeInt(Minimum=24, Maximum=32)
    Spawn Group=1
    Use Spawn Probability=false
    UseLoopCountLimit=false
   ) {module=SpawnBurst_Instantaneous001}
  ]
  particleSpawnScript [
   /SetVariables(Particles.VisibilityTag=/Niagara/DynamicInputs/Helpers/MaskInt(Mask Parameter:NiagaraInt32=$Emitter.SpawnGroup, Mask:NiagaraInt32=0, Masked Int:NiagaraInt32=0, Passthrough Int:NiagaraInt32=1)) {isSetParameters=true}
   &InitializeParticle(
    Color Maximum=<rgba 0.9594 0.3649 1 1>
    Color Minimum=<rgba 1.4323 4.5309 5 1>
    Color Mode=::NewEnumerator2
    Lifetime=1.6
    Mass Mode=::NewEnumerator1
    Sprite Rotation Mode=::NewEnumerator2
    Sprite Size Mode=::NewEnumerator1
    Sprite UV Mode=::NewEnumerator3
    Uniform Sprite Size Max=16
    Uniform Sprite Size Min=3
    UsePositionOffset=false
    Write Material Random=true
   )
   &ShapeLocation(
    Handle Radius=40
    Large Radius=70
    MaskSpawnGroup=true
    Override Seed=false
    Rotation Axis=<0 0 1>
    Rotation Mode=::NewEnumerator1
    Shape Primitive=::NewEnumerator3
    Spawn Group Mask=1
    U Distribution=0.5
    UseOwnerScale=true
   )
   &SkeletalMeshLocation(
    Bone Sampling Mode=::NewEnumerator0
    MaskSpawnGroup=false
    OverrideRandomSeed=false
    Skeletal Mesh=$System.Skeletal Mesh
    Write To Particle Position:NiagaraBool=/Niagara/DynamicInputs/Helpers/MaskBoolBySpawnGroup(Masked Bool:NiagaraBool=true, Passthrough Bool:NiagaraBool=false, Spawn Group Mask:NiagaraInt32=0)
   )
   /SetVariables(Particles.EQPosition:NiagaraPosition=$Particles.ShapeLocation.ShapePosition) {enabled=false, isSetParameters=true}
   &InitialMeshOrientation(Use Rotation Vector=false)
   &AddVelocity(
    Module.Use Falloff Distance=false
    Origin Offset=<0 0 -100>
    Override Seed=false
    Velocity Mode=::NewEnumerator1
    Velocity Speed=&RandomRangeFloat(Minimum=45, Maximum=300)
   )
  ]
  particleUpdateScript [
   &ParticleState()
   &CurlNoiseForce(Bias Noise=false, Noise Strength=70, Pan Noise=false)
   &Drag(
    Drag=&FloatFromCurve(
     FloatCurve=DataInterface() {propertyValues {Curve {keys [
         {arriveTangent=5.8418, interpMode=RCIM_Cubic, leaveTangent=5.8418, tangentMode=RCTM_User, time=0, value=0}
         {arriveTangent=-0.0011, interpMode=RCIM_Cubic, leaveTangent=-0.0011, tangentMode=RCTM_User, time=0.2, value=1}
         {arriveTangent=0.0415, interpMode=RCIM_Cubic, leaveTangent=0.0415, tangentMode=RCTM_User, time=0.9, value=1}
         <at 1 0>
        ]}}}
     Scale Curve=&RandomRangeFloat(Minimum=3, Maximum=5)
    )
    Rotational Drag:NiagaraFloat=&FloatFromCurve(
     FloatCurve=DataInterface() {propertyValues {Curve {keys [
         {arriveTangent=5.8418, interpMode=RCIM_Cubic, leaveTangent=5.8418, tangentMode=RCTM_User, time=0, value=0}
         {arriveTangent=-0.0011, interpMode=RCIM_Cubic, leaveTangent=-0.0011, tangentMode=RCTM_User, time=0.2, value=1}
         {arriveTangent=0.0415, interpMode=RCIM_Cubic, leaveTangent=0.0415, tangentMode=RCTM_User, time=0.7, value=1}
         {arriveTangent=0.0324, interpMode=RCIM_Cubic, leaveTangent=0.0324, tangentMode=RCTM_User, time=0.75, value=0}
         <at 1 0>
        ]}}}
     Scale Curve=&RandomRangeFloat(Minimum=15, Maximum=15)
    )
    Use Linear Drag=true
    Use Rotational Drag=true
   )
   &PointAttractionForce(
    Attraction Radius=500
    Attraction Strength=&FloatFromCurve(
     FloatCurve=DataInterface() {propertyValues {Curve {keys [<at 0 0 RCIM_Constant>,<at 0.675 1 RCIM_Constant>,<at 1 1>]}}}
     Scale Curve=&RandomRangeFloat(Minimum=325, Maximum=390)
    )
    Kill Radius:NiagaraFloat=&MaskFloatBySpawnGroup(Masked Float=4, Passthrough Float=80)
    Kill Within Radius=true
    Module.Use Falloff=true
    UseAttractorOffset=false
   )
   &SpriteRotationRate(Rotation Rate=&FloatFromCurve(
     FloatCurve=DataInterface() {propertyValues {Curve {keys [<at 0 0.2>,<at 0.75 0.2>,<at 1 1>]}}}
     Scale Curve=&RandomRangeFloat(Minimum=500, Maximum=1500)
    )) {module=SpriteRotationRate001}
   &SolveForcesAndVelocity(Clamp Velocity=false, Limit Acceleration=false)
   &ScaleSpriteSize(Uniform Curve Sprite Scale=DataInterface() {propertyValues {Curve {keys [
        {arriveTangent=25.2952, interpMode=RCIM_Cubic, leaveTangent=25.2952, tangentMode=RCTM_User, time=0, value=0}
        {arriveTangent=-0.0259, interpMode=RCIM_Cubic, leaveTangent=-0.0259, tangentMode=RCTM_User, time=0.1, value=1}
        {arriveTangent=-0.0039, interpMode=RCIM_Cubic, leaveTangent=-0.0039, tangentMode=RCTM_User, time=0.5, value=1}
        {arriveTangent=-0.0288, interpMode=RCIM_Cubic, leaveTangent=-0.0288, tangentMode=RCTM_User, time=1, value=0}
       ]}}})
   &ScaleColor(
    Linear Color Curve=DataInterface() {propertyValues {
      AlphaCurve {keys [<at 0 1>]}
      BlueCurve {keys [<at 0 1>,<at 0.5 1>,<at 0.8 5>]}
      GreenCurve {keys [<at 0 1>,<at 0.5 1>,<at 0.8 5>]}
      RedCurve {keys [<at 0 1>,<at 0.5 1>,<at 0.8 5>]}
     }}
    Scale Mode=::NewEnumerator2
   )
  ]
  rendererClasses [&NiagaraSpriteRendererProperties]
  renderers [
   {rendererClass=&NiagaraSpriteRendererProperties}
   {rendererClass=&NiagaraSpriteRendererProperties, rendererIndex=1}
  ]
 }
 Halo_Rays_Glow_Wave {
  emitterUpdateScript [
   &EmitterState()
   &SpawnBurst_Instantaneous(Use Spawn Probability=false, UseLoopCountLimit=false)
   &SpawnBurst_Instantaneous(Spawn Group=1, Use Spawn Probability=false, UseLoopCountLimit=false) {module=SpawnBurst_Instantaneous001}
  ]
  particleSpawnScript [
   &InitializeParticle(
    Color=<rgba 1 1 1 0.5>
    Lifetime=&MaskFloatBySpawnGroup(Masked Float=1.2, Passthrough Float=1.6)
    Mesh Scale Mode=::NewEnumerator3
    Position Offset=$User.Halo Position
    Sprite Size Mode=::NewEnumerator3
    Uniform Sprite Size=200
    UsePositionOffset=true
    Write Material Random=true
   )
   /SetVariables(Particles.VisibilityTag=&MaskIntBySpawnGroup(Masked Int=0, Passthrough Int=1, Spawn Group Mask=0)) {isSetParameters=true}
   /SetVariables(Particles.GlowCO=10) {isSetParameters=true}
   &InitialMeshOrientation(
    Orientation Vector=<0 0 1>
    Rotation=&RandomRangeVector(Minimum=<0 0 0>, Maximum=<0 0 1>)
    Use Orientation Vector=true
    Use Rotation Vector=true
   )
  ]
  particleUpdateScript [
   &ParticleState()
   &SolveForcesAndVelocity(Clamp Velocity=false, Limit Acceleration=false)
   &ScaleMeshSize(Scale Factor=&VectorFromFloat(Value=&FloatFromCurve(FloatCurve=DataInterface() {propertyValues {Curve {keys [<at 0 0>,<at 0.1 1.4>,<at 0.85 2>,<at 1 0>]}}})))
   &ScaleColor(
    Linear Color Curve=DataInterface() {propertyValues {
      AlphaCurve {keys [<at 0 0 RCIM_Cubic>,<at 0.5 1>,<at 0.8 1>,<at 1 0 RCIM_Cubic>]}
      BlueCurve {keys [<at 0 2.6146>,<at 0.1856 1.7807>,<at 0.4289 1.3224>,<at 0.6 1>]}
      GreenCurve {keys [<at 0 0>,<at 0.1856 0.1114>,<at 0.4289 0.198>,<at 0.6 0.3661>]}
      RedCurve {keys [<at 0 3>,<at 0.1856 0.459>,<at 0.4289 0.1393>,<at 0.6 0.0075>]}
     }}
    Scale Mode=::NewEnumerator2
   )
   &ScaleSpriteSize(Scale Factor=&Multiply_Vector2D_ByFloat(
     Float=&FloatFromCurve(
      FloatCurve=DataInterface() {propertyValues {Curve {keys [{interpMode=RCIM_Cubic, tangentMode=RCTM_User, time=0, value=0.5}, {interpMode=RCIM_Cubic, tangentMode=RCTM_User, time=1, value=1}]}}}
      Scale Curve=1.25
     )
     Vector2D=<1 1>
    ))
   /SetVariables(Particles.Corona_Size:Vector2f=&Multiply_Vector2D_ByFloat(Float=0.5, Vector2D=$Particles.SpriteSize)) {isSetParameters=true}
   /SetVariables(
    Particles.GlowColor=&ColorFromCurve(ColorCurve=DataInterface() {propertyValues {
       AlphaCurve {keys [<at 0 1 RCIM_Cubic>,<at 0.6 1>,<at 0.9 0>]}
       BlueCurve {keys [<at 0 3>,<at 0.6 1>]}
       GreenCurve {keys [<at 0 0.8906>,<at 0.6 0>]}
       RedCurve {keys [<at 0 0>,<at 0.6 0.3051>]}
      }})
    Particles.GlowSize=&Vector2DFromFloat(Value=&FloatFromCurve(
      FloatCurve=DataInterface() {propertyValues {Curve {keys [
          {arriveTangent=11.8116, interpMode=RCIM_Cubic, leaveTangent=11.8116, tangentMode=RCTM_User, time=0, value=0}
          {arriveTangent=-0.0883, interpMode=RCIM_Cubic, leaveTangent=-0.0883, tangentMode=RCTM_User, time=0.1, value=1.2}
          {arriveTangent=-0.2122, interpMode=RCIM_Cubic, leaveTangent=-0.2122, tangentMode=RCTM_User, time=0.25, value=1}
          {interpMode=RCIM_Cubic, leaveTangent=-2.2807, time=1, value=1}
         ]}}}
      Scale Curve=200
     ))
   ) {isSetParameters=true}
   /SetVariables(
    Particles.WaveColor:LinearColor=&MultiplyLinearColor(
     Linear Color=&ColorFromCurve(ColorCurve=DataInterface() {propertyValues {
        AlphaCurve {keys [<at 0 1>,<at 1 0>]}
        BlueCurve {keys [<at 0 10>,<at 0.15 3>,<at 1 1>]}
        GreenCurve {keys [<at 0 10>,<at 0.15 3>,<at 1 1>]}
        RedCurve {keys [<at 0 10>,<at 0.15 3>,<at 1 1>]}
       }})
     Scale Factor=&MakeVector4_FromLinearColor(LinearColor=<rgba 0 2.8826 5 0.05>)
    )
    Particles.WaveDynamicMaterialParameter:Vector4f=&MakeVector4(
     X=&FloatFromCurve(FloatCurve=DataInterface() {propertyValues {Curve {keys [{arriveTangent=-2.1015, interpMode=RCIM_Cubic, leaveTangent=-2.1015, tangentMode=RCTM_User, time=0, value=0.5}, <at 1 -0.15>]}}})
     Y=&FloatFromCurve(FloatCurve=DataInterface() {propertyValues {Curve {keys [{arriveTangent=-0.3464, interpMode=RCIM_Cubic, leaveTangent=-0.3464, tangentMode=RCTM_User, time=0, value=1}, <at 1 0.9>]}}})
     Z=&FloatFromCurve(FloatCurve=DataInterface() {propertyValues {Curve {keys [<at 0 0>,<at 0.15 1>,<at 1 1.2>]}}}, Scale Curve=0.8)
     W=0
    )
   ) {isSetParameters=true}
  ]
  rendererClasses [&NiagaraSpriteRendererProperties, &NiagaraMeshRendererProperties]
  renderers [
   {rendererClass=&NiagaraSpriteRendererProperties}
   {rendererClass=&NiagaraSpriteRendererProperties, rendererIndex=1}
   {rendererClass=&NiagaraSpriteRendererProperties, rendererIndex=2}
   {rendererClass=&NiagaraMeshRendererProperties, rendererIndex=3}
  ]
 }
 Hologram_Lines {
  emitterUpdateScript [
   &EmitterState()
   &SpawnRate(
    SpawnRate=&FloatFromCurve(
     FloatCurve=DataInterface() {propertyValues {Curve {keys [
         {arriveTangent=-286.7, interpMode=RCIM_Cubic, leaveTangent=-286.7, tangentMode=RCTM_User, time=0, value=150}
         {arriveTangent=-85.5422, tangentMode=RCTM_User, time=0.7, value=0}
        ]}}}
     CurveIndex=$Emitter.NormalizedLoopAge
     Scale Curve=&FN_FrontEndFloatModifier(Gameplay=0.25, Front End=0.6)
    )
    Use Spawn Probability=false
   )
  ]
  particleSpawnScript [
   &InitializeParticle(
    Color=<rgba 3 3 3 0.2>
    Lifetime Max=0.7
    Lifetime Min=0.4
    Lifetime Mode=::NewEnumerator1
    Mesh Scale Max=<1 1 20>
    Mesh Scale Min=<0.1 0.1 5>
    Mesh Scale Mode=::NewEnumerator2
    Position Offset=$User.Halo Position
    Sprite Size Max=<5 50>
    Sprite Size Min=<3 35>
    Sprite Size Mode=::NewEnumerator2
    UsePositionOffset=true
    Write Material Random=true
   )
   &ShapeLocation(
    Handle Radius=35
    Large Radius=85
    MaskSpawnGroup=false
    Non Uniform Scale=<1 1 3>
    Override Seed=false
    Rotation Axis=<0 0 1>
    Rotation Mode=::NewEnumerator1
    Shape Primitive=::NewEnumerator3
    U Distribution=0.5
    UseOwnerScale=true
   )
   &CameraOffset(Camera Offset Amount=15)
   &AddVelocity(
    Module.Use Falloff Distance=false
    Origin Offset=<0 0 70>
    Override Seed=false
    Velocity Mode=::NewEnumerator1
    Velocity Speed=&RandomRangeFloat(Minimum=55, Maximum=75)
   ) {module=AddVelocity001}
  ]
  particleUpdateScript [
   &ParticleState()
   &Drag(Drag=5, Use Linear Drag=true, Use Rotational Drag=true)
   &SolveForcesAndVelocity(Clamp Velocity=false, Limit Acceleration=false)
   &ScaleColor(
    Scale Alpha=&FloatFromCurve(FloatCurve=DataInterface() {propertyValues {Curve {keys [{interpMode=RCIM_Cubic, tangentMode=RCTM_User, time=0, value=1}, {interpMode=RCIM_Cubic, tangentMode=RCTM_User, time=1, value=0}]}}})
    Scale RGB=&Multiply_VectorByFloat(
     Vector=&MakeVectorFromLinearColorRGB(Color=&ColorFromCurve(ColorCurve=DataInterface() {propertyValues {
         AlphaCurve {keys []}
         BlueCurve {keys [<at 0.0799 0.9944>,<at 0.2471 1>,<at 0.3784 1>,<at 0.5465 0.0134>,<at 0.6755 1>,<at 0.8368 1>]}
         GreenCurve {keys [<at 0.0799 1>,<at 0.2471 0.9713>,<at 0.3784 0.9268>,<at 0.5465 1>,<at 0.6755 0.8361>,<at 0.8368 0.5791>]}
         RedCurve {keys [<at 0.0799 0>,<at 0.2471 0.1195>,<at 0.3784 0>,<at 0.5465 0.2875>,<at 0.6755 0>,<at 0.8368 0>]}
        }}))
     Float=&FloatFromCurve(
      FloatCurve=DataInterface() {propertyValues {Curve {keys [
          {arriveTangent=-9.3569, interpMode=RCIM_Cubic, leaveTangent=-9.3569, tangentMode=RCTM_User, time=0, value=2}
          {arriveTangent=-2.1612, interpMode=RCIM_Cubic, leaveTangent=-2.1612, tangentMode=RCTM_User, time=0.1, value=1}
          <at 1 0.3>
         ]}}}
      CurveIndex=$Emitter.Age
     )
    )
    ScaleA=true
    ScaleRGB=true
   )
  ]
  rendererClasses [&NiagaraSpriteRendererProperties]
  renderers [{rendererClass=&NiagaraSpriteRendererProperties}]
 }
 Streaks_Body {
  emitterUpdateScript [
   &EmitterState()
   &SpawnBurst_Instantaneous(
    Spawn Count=&RandomRangeInt(Minimum=8, Maximum=12)
    Use Spawn Probability=false
    UseLoopCountLimit=false
   )
  ]
  particleSpawnScript [
   &InitializeParticle(
    Color=<rgba 5 5 5 1>
    Lifetime Max=0.35
    Lifetime Min=0.12
    Lifetime Mode=::NewEnumerator1
    Sprite Size Max=<25 65>
    Sprite Size Min=<10 45>
    Sprite Size Mode=::NewEnumerator2
    UsePositionOffset=false
    Write Material Random=true
   )
   &SubUVAnimation(End Frame=3, SubUV Animation Mode:ENiagara_SubUVLookupMode=::NewEnumerator2)
   &AddVelocity(
    Cone Angle=65
    Cone Axis=<0 0 1>
    Module.UseConeFalloff=false
    Module.UseVelDistribution=false
    Override Seed=false
    Velocity Mode=::NewEnumerator2
    Velocity Speed=&RandomRangeFloat(Minimum=175, Maximum=225)
   )
   &CameraOffset(Camera Offset Amount=-50)
  ]
  particleUpdateScript [
   &SolveForcesAndVelocity(Clamp Velocity=false, Limit Acceleration=false)
   &ParticleState()
   &ScaleColor(
    Scale Alpha=&FloatFromCurve(FloatCurve=DataInterface() {propertyValues {Curve {keys [{arriveTangent=-3.3102, interpMode=RCIM_Cubic, leaveTangent=-3.3102, tangentMode=RCTM_User, time=0, value=1}, <at 1 0 RCIM_Cubic>]}}})
    Scale RGB=&MakeVectorFromLinearColorRGB(Color=&ColorFromCurve(ColorCurve=DataInterface() {propertyValues {
        AlphaCurve {keys []}
        BlueCurve {keys [<at 0 10 RCIM_Cubic>,<at 0.125 1>,<at 1 0.1878 RCIM_Cubic>]}
        GreenCurve {keys [<at 0 10 RCIM_Cubic>,<at 0.125 1>,<at 1 0.1878 RCIM_Cubic>]}
        RedCurve {keys [<at 0 5.481 RCIM_Cubic>,<at 0.125 0.1661>,<at 1 0.0192 RCIM_Cubic>]}
       }}))
    ScaleA=true
    ScaleRGB=true
   )
   &ScaleSpriteSize(
    Non-Uniform Curve Sprite Scale=DataInterface() {propertyValues {
      XCurve {keys [
        {arriveTangent=0.9067, interpMode=RCIM_Cubic, leaveTangent=0.9067, tangentMode=RCTM_User, time=0, value=0.8}
        {arriveTangentWeight=0.2034, interpMode=RCIM_Cubic, leaveTangentWeight=0.2034, tangentMode=RCTM_User, time=0.5, value=1}
        {arriveTangent=-3.1509, interpMode=RCIM_Cubic, leaveTangent=-3.1509, tangentMode=RCTM_User, time=1, value=0.1}
       ]}
      YCurve {keys [
        {arriveTangent=3.7985, interpMode=RCIM_Cubic, leaveTangent=3.7985, tangentMode=RCTM_User, time=0, value=0}
        {arriveTangent=2, interpMode=RCIM_Cubic, leaveTangent=2, time=0.5, value=1.6}
        <at 1 2 RCIM_Cubic>
       ]}
     }}
    Scale Sprite Size Mode=::NewEnumerator3
   )
  ]
  rendererClasses [&NiagaraSpriteRendererProperties]
  renderers [{rendererClass=&NiagaraSpriteRendererProperties}]
 }
 Streaks_GlassOut {
  emitterUpdateScript [
   &EmitterState()
   &SpawnBurst_Instantaneous(
    Spawn Count=&RandomRangeInt(Minimum=12, Maximum=16)
    Use Spawn Probability=false
    UseLoopCountLimit=false
   )
   &SpawnBurst_Instantaneous(
    Spawn Count=&RandomRangeInt(Minimum=8, Maximum=12)
    Spawn Group=1
    Spawn Time=1.15
    Use Spawn Probability=false
    UseLoopCountLimit=false
   ) {module=SpawnBurst_Instantaneous001}
  ]
  particleSpawnScript [
   /SetVariables(Particles.VisibilityTag=&MaskIntBySpawnGroup(Masked Int=0, Passthrough Int=1, Spawn Group Mask=0)) {isSetParameters=true}
   &InitializeParticle(
    Color=<rgba 5 5 5 1>
    Lifetime Max=0.35
    Lifetime Min=0.12
    Lifetime Mode=::NewEnumerator1
    Sprite Size Max=<5 40>
    Sprite Size Min=<3 25>
    Sprite Size Mode=::NewEnumerator2
    UsePositionOffset=false
    Write Material Random=true
   )
   &ShapeLocation(
    Handle Radius=20
    Large Radius=&MaskFloatBySpawnGroup(Masked Float=80, Passthrough Float=95)
    MaskSpawnGroup=false
    Override Seed=false
    Rotation Axis=<0 0 1>
    Rotation Mode=::NewEnumerator1
    Shape Primitive=::NewEnumerator3
    U Distribution=0.5
    UseOwnerScale=true
   )
   &ShapeLocation(
    MaskSpawnGroup=false
    Override Seed=false
    Sphere Radius=&MaskFloatBySpawnGroup(Masked Float=0, Passthrough Float=50)
    UseOwnerScale=true
   ) {module=ShapeLocation001}
   &SubUVAnimation(End Frame=3, SubUV Animation Mode:ENiagara_SubUVLookupMode=::NewEnumerator2)
   &AddVelocity(
    Module.Use Falloff Distance=false
    Override Seed=false
    Velocity Mode=::NewEnumerator1
    Velocity Speed=&MaskFloatBySpawnGroup(Masked Float=&RandomRangeFloat(Minimum=75, Maximum=150), Passthrough Float=0)
   )
  ]
  particleUpdateScript [
   &PointAttractionForce(
    Attraction Radius=400
    Attraction Strength=&MaskFloatBySpawnGroup(Masked Float=0, Passthrough Float=&RandomRangeFloat(Minimum=120, Maximum=250))
    Kill Within Radius=true
    Module.Use Falloff=true
    UseAttractorOffset=false
   )
   &SolveForcesAndVelocity(Clamp Velocity=false, Limit Acceleration=false)
   &ParticleState()
   &ScaleColor(
    Scale Alpha=&FloatFromCurve(FloatCurve=DataInterface() {propertyValues {Curve {keys [{arriveTangent=-3.3102, interpMode=RCIM_Cubic, leaveTangent=-3.3102, tangentMode=RCTM_User, time=0, value=1}, <at 1 0 RCIM_Cubic>]}}})
    Scale RGB=&MakeVectorFromLinearColorRGB(Color=&ColorFromCurve(ColorCurve=DataInterface() {propertyValues {
        AlphaCurve {keys []}
        BlueCurve {keys [<at 0 10 RCIM_Cubic>,<at 0.125 1>,<at 1 0.1878 RCIM_Cubic>]}
        GreenCurve {keys [<at 0 10 RCIM_Cubic>,<at 0.125 1>,<at 1 0.1878 RCIM_Cubic>]}
        RedCurve {keys [<at 0 5.481 RCIM_Cubic>,<at 0.125 0.1661>,<at 1 0.0192 RCIM_Cubic>]}
       }}))
    ScaleA=true
    ScaleRGB=true
   )
   &ScaleSpriteSize(
    Non-Uniform Curve Sprite Scale=DataInterface() {propertyValues {
      XCurve {keys [
        {arriveTangent=0.9067, interpMode=RCIM_Cubic, leaveTangent=0.9067, tangentMode=RCTM_User, time=0, value=0.8}
        {arriveTangentWeight=0.2034, interpMode=RCIM_Cubic, leaveTangentWeight=0.2034, tangentMode=RCTM_User, time=0.5, value=1}
        {arriveTangent=-3.1509, interpMode=RCIM_Cubic, leaveTangent=-3.1509, tangentMode=RCTM_User, time=1, value=0.1}
       ]}
      YCurve {keys [
        {arriveTangent=3.7985, interpMode=RCIM_Cubic, leaveTangent=3.7985, tangentMode=RCTM_User, time=0, value=0}
        {arriveTangent=2, interpMode=RCIM_Cubic, leaveTangent=2, time=0.5, value=1.6}
        <at 1 2 RCIM_Cubic>
       ]}
     }}
    Scale Sprite Size Mode=::NewEnumerator3
   )
   /SetVariables(
    Particles.GlassInColor:LinearColor=&MultiplyLinearColor(
     Linear Color=&ColorFromCurve(ColorCurve=DataInterface() {propertyValues {
        AlphaCurve {keys [{arriveTangent=-3.3102, interpMode=RCIM_Cubic, leaveTangent=-3.3102, tangentMode=RCTM_User, time=0, value=1}, <at 1 0 RCIM_Cubic>]}
        BlueCurve {keys [<at 0 10 RCIM_Cubic>,<at 0.125 1>,<at 1 0.1878 RCIM_Cubic>]}
        GreenCurve {keys [<at 0 10 RCIM_Cubic>,<at 0.125 1>,<at 1 0.1878 RCIM_Cubic>]}
        RedCurve {keys [<at 0 6.9792 RCIM_Cubic>,<at 0.125 0.5365>,<at 1 0.0192 RCIM_Cubic>]}
       }})
     Scale Factor=&MakeVector4_FromLinearColor(LinearColor=<rgba 5 5 5 0.35>)
    )
    Particles.GlassInSpriteSize:Vector2f=&Vector2DFromCurve(
     Vector2Curve=DataInterface() {propertyValues {
       XCurve {keys [
         {arriveTangent=0.9067, interpMode=RCIM_Cubic, leaveTangent=0.9067, tangentMode=RCTM_User, time=0, value=0.8}
         {arriveTangentWeight=0.2034, interpMode=RCIM_Cubic, leaveTangentWeight=0.2034, tangentMode=RCTM_User, time=0.5, value=1}
         {arriveTangent=-3.1509, interpMode=RCIM_Cubic, leaveTangent=-3.1509, tangentMode=RCTM_User, time=1, value=0.1}
        ]}
       YCurve {keys [
         {arriveTangent=3.7985, interpMode=RCIM_Cubic, leaveTangent=3.7985, tangentMode=RCTM_User, time=0, value=0}
         {arriveTangent=2, interpMode=RCIM_Cubic, leaveTangent=2, time=0.5, value=1.6}
         <at 1 2 RCIM_Cubic>
        ]}
      }}
     ScaleCurve=&RandomRangeVector2D(Minimum=<3 55>, Maximum=<5 75>)
    )
   ) {isSetParameters=true}
  ]
  rendererClasses [&NiagaraSpriteRendererProperties]
  renderers [
   {rendererClass=&NiagaraSpriteRendererProperties}
   {rendererClass=&NiagaraSpriteRendererProperties, rendererIndex=1}
  ]
 }
}
summary {
 systemName=NS_Emote_Omega_Debris
 emitters [
  {emitterName=Glass_Body_EQ, rendererClasses [&NiagaraSpriteRendererProperties]}
  {emitterName=Hologram_Lines, rendererClasses [&NiagaraSpriteRendererProperties]}
  {emitterName=Cracks, rendererClasses [&NiagaraMeshRendererProperties, &NiagaraSpriteRendererProperties]}
  {emitterName=Streaks_Body, rendererClasses [&NiagaraSpriteRendererProperties]}
  {emitterName=Streaks_GlassOut, rendererClasses [&NiagaraSpriteRendererProperties]}
  {
   emitterName=Halo_Rays_Glow_Wave
   rendererClasses [&NiagaraSpriteRendererProperties, &NiagaraMeshRendererProperties]
  }
 ]
 userVariables [
  User.bIsFrontEnd {type=NiagaraBool}
  User.bIsFrontEndPreview {type=NiagaraBool}
  User.Crack Position {type=Vector3f, description="", defaultValue {struct=Vector3f, value=<0 0 0>}}
  User.Halo Position {type=Vector3f, description="", defaultValue {struct=Vector3f, value=<0 0 0>}}
  User.VariantSwitch {type=NiagaraFloat, defaultValue=0, description=""}
 ]
}
