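# Niagara one-shot burst effect for the Omega emote, named for an instrument burst-up.
# EnergyWave_Burst spawns a growing, color-cycling mesh at the skeletal mesh's root bone, tinted from a color palette, with material parameters varying by platform and front-end context.
# Splash_Foam bursts a few randomly tinted, spinning SubUV sprites in an upward-axis cone that flatten and fade.
# Droplets001, Glow001 and Impact_Streaks001 list no renderers or modules here.
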NiagaraSystem /BRCosmetics/Animation/Game/MainPlayer/Emotes/Omega/FX/NS_Emote_Omega_InstrumentBurstUp
dependencies {
 usedDataInterfaces [
  &NiagaraDataInterfaceVector2DCurve
  &NiagaraDataInterfaceColorCurve
  &NiagaraDataInterfaceCurve
  &NiagaraDataInterfaceArrayColor
  &NiagaraDataInterfaceSkeletalMesh
  &NiagaraDataInterfacePlatformSet
 ]
 usedDynamicInputs [&RandomRangeVector]
 usedModules [&SystemState]
}
emitters {
 Droplets001 {rendererClasses [], renderers []}
 EnergyWave_Burst {
  emitterUpdateScript [
   &EmitterState()
   &SpawnBurst_Instantaneous(Use Spawn Probability=false, UseLoopCountLimit=false)
  ]
  particleSpawnScript [
   &InitializeParticle(
    Color=&SelectLinearColorFromArray(
     Array Sampling Mode=::NewEnumerator0
     Direct Array Index=?
     Color Selection Array=DataInterface() {propertyValues {ColorData [<rgba 1 0.2172 0.1979 1>, <rgba 0.6561 0.1667 1 1>, <rgba 1 0.9512 0.1771 1>, <rgba 0.1927 0.9557 1 1>, <rgba 0.2135 0.3889 1 1>]}}
     Random Seed=0
    )
    Lifetime=0.6
    Mesh Scale Mode=::NewEnumerator3
    Mesh Uniform Scale=1.3
    Position Offset=<0 0 -80>
    UsePositionOffset=false
    Write Material Random=true
   )
   &SkeletalMeshLocation(
    Bone Sampling Mode=::NewEnumerator4
    MaskSpawnGroup=false
    Orientation Sampling=::NewEnumerator1
    Skeletal Mesh=DataInterface() {propertyValues {
      DefaultMesh=null
      ExcludeBoneName=root
      OptionalComponentClass=null
      PreviewMesh=/Game/Characters/Player/Male/Male_Avg_Base/Fortnite_M_Avg_Player
      SkinningMode=SkinOnTheFly
      SoftSourceActor=null
      SourceMode=Default
      UvSetIndex=0
      WholeMeshLOD=-1
      bExcludeBone=false
      bReadDeformedGeometry=true
      bRequireCurrentFrameData=true
      ComponentTags []
      FilteredBones [root]
      FilteredSockets []
      MeshUserParameter {parameter {name=None, typeDefHandle {registeredTypeIndex=1}, varData []}}
      SamplingRegions []
     }}
   )
  ]
  particleUpdateScript [
   &ParticleState()
   &ScaleMeshSize(Scale Factor=&VectorFromFloat(Value=&FloatFromCurve(FloatCurve=DataInterface() {propertyValues {Curve {keys [<at 0 0 RCIM_Cubic>, {arriveTangent=1, interpMode=RCIM_Cubic, leaveTangent=1, time=0.2041, value=0.6997}, <at 1 1 RCIM_Cubic>]}}})))
   &ScaleColor(
    Linear Color Curve=DataInterface() {propertyValues {
      AlphaCurve {keys [<at 0 1>,<at 1 0>]}
      BlueCurve {keys [<at 0.1259 2 RCIM_Cubic>,<at 0.2277 1.8773>,<at 0.4 0>,<at 0.5655 0>,<at 0.75 0.2851 RCIM_Cubic>,<at 1 0>]}
      GreenCurve {keys [
        <at 0.1259 1.1188 RCIM_Cubic>
        <at 0.2277 3>
        <at 0.4 1.9095>
        <at 0.5655 0.9115>
        {arriveTangent=0.2037, interpMode=RCIM_Cubic, leaveTangent=0.2037, time=0.75, value=0}
        <at 1 1>
       ]}
      RedCurve {keys [
        <at 0.1259 0.6458 RCIM_Cubic>
        <at 0.2277 1.2201>
        <at 0.4 2.5926>
        <at 0.5655 1.6785>
        {arriveTangent=-1.6002, interpMode=RCIM_Cubic, leaveTangent=-1.6002, time=0.75, value=2}
        <at 1 0.9833>
       ]}
     }}
    Scale Mode=::NewEnumerator2
   )
   &DynamicMaterialParameters(
    Index 0 Param 1=&FloatFromCurve(FloatCurve=DataInterface() {propertyValues {Curve {keys [<at 0 1>,<at 0.2 1.375>,<at 1 0>]}}}, Scale Curve=1.2)
    Index 0 Param 2=2
    Index 0 Param 3=&FloatFromCurve(
     FloatCurve=DataInterface() {propertyValues {Curve {keys [<at 0.1 1>,<at 1 0>]}}}
     Scale Curve=&FN_FrontEndFloatModifier(Gameplay=4, Front End=3)
    )
    Index 0 Param 4=&FloatFromCurve(FloatCurve=DataInterface() {propertyValues {Curve {keys [
         <at 0 1>
         {arriveTangent=0.3079, interpMode=RCIM_Cubic, leaveTangent=0.3079, tangentMode=RCTM_User, time=0.3, value=0.9}
         <at 1 1 RCIM_Cubic>
        ]}}})
    Index 1 Param 1=&FN_PickFloatDependOnPlatform(TargetPlatformActive=180, TargetPlatformNoActive=100)
    Index 1 Param 2=&FN_PickFloatDependOnPlatform(TargetPlatformActive=2.5, TargetPlatformNoActive=5)
    Module.Index0Param0Write=true
    Module.Index0Param1Write=true
    Module.Index0Param2Write=true
    Module.Index0Param3Write=true
    Module.Index1Param0Write=true
    Module.Index1Param1Write=true
    Module.Index1Param2Write=true
    Module.Index1Param3Write=true
    Write Parameter Index 1=true
   )
  ]
  rendererClasses [&NiagaraMeshRendererProperties]
  renderers [{rendererClass=&NiagaraMeshRendererProperties, rendererIndex=0}]
 }
 Glow001 {rendererClasses [], renderers []}
 Impact_Streaks001 {rendererClasses [], renderers []}
 Splash_Foam {
  emitterUpdateScript [
   &EmitterState()
   &SpawnBurst_Instantaneous(Spawn Count=5, Use Spawn Probability=false, UseLoopCountLimit=false)
   &SpawnRate(SpawnRate=18, Use Spawn Probability=false) {enabled=false}
  ]
  particleSpawnScript [
   &InitializeParticle(
    Color Maximum=<rgba 5 0.4948 1.2765 1>
    Color Minimum=<rgba 2.0329 0.599 5 1>
    Color Mode=::NewEnumerator2
    Lifetime Max=0.7
    Lifetime Min=0.4
    Lifetime Mode=::NewEnumerator1
    Mesh Scale Mode=::NewEnumerator3
    Position Offset=<0 0 75>
    Sprite Rotation Mode=::NewEnumerator2
    Sprite Size Max=<25 35>
    Sprite Size Min=<10 25>
    Sprite Size Mode=::NewEnumerator2
    UsePositionOffset=false
    Write Material Random=true
   ) {module=InitializeParticle001}
   /Niagara/Modules/Update/SubUV/SubUVAnimation(SubUV Animation Mode=::NewEnumerator1)
   &AddVelocity(
    Cone Angle=90
    Cone Axis=<0 0 1>
    Module.UseConeFalloff=false
    Module.UseVelDistribution=false
    Override Seed=false
    Velocity Mode=::NewEnumerator2
    Velocity Speed=&RandomRangeFloat(Minimum=35, Maximum=90)
   )
   &CameraOffset(Camera Offset Amount=-25)
  ]
  particleUpdateScript [
   &ParticleState()
   &SpriteRotationRate(Rotation Rate=&RandomRangeFloat(Minimum=-90, Maximum=90))
   &Drag(Drag=0.25, Use Linear Drag=true, Use Rotational Drag=false)
   &SolveForcesAndVelocity(Clamp Velocity=false, Limit Acceleration=false)
   &ScaleSpriteSize(Scale Factor=&Vector2DFromCurve(
     Vector2Curve=DataInterface() {propertyValues {
       XCurve {keys [<at 0 1>,<at 1 1>]}
       YCurve {keys [<at 0 1 RCIM_Cubic>, {arriveTangent=-3.2254, interpMode=RCIM_Cubic, leaveTangent=-3.2254, tangentMode=RCTM_User, time=1, value=0}]}
      }}
     ScaleCurve=<1 1>
    ))
   &ScaleColor(
    Linear Color Curve=DataInterface() {propertyValues {
      AlphaCurve {keys [<at 0 1>,<at 0.8196 1>,<at 1 0>]}
      BlueCurve {keys [<at 0 1>,<at 1 0.2>]}
      GreenCurve {keys [<at 0 1>,<at 1 0.2>]}
      RedCurve {keys [<at 0 1>,<at 1 0.2>]}
     }}
    Scale Mode=::NewEnumerator2
   )
   &DynamicMaterialParameters(
    Index 0 Param 1=&RandomRangeFloat(Minimum=0, Maximum=1)
    Index 0 Param 3=&FloatFromCurve(FloatCurve=DataInterface() {propertyValues {Curve {keys [{interpMode=RCIM_Cubic, tangentMode=RCTM_User, time=0, value=0}, {interpMode=RCIM_Cubic, tangentMode=RCTM_User, time=0.2, value=1}]}}})
    Index 0 Param 4=&FloatFromCurve(FloatCurve=DataInterface() {propertyValues {Curve {keys [
         {arriveTangent=2.5113, interpMode=RCIM_Cubic, time=0, value=0}
         {arriveTangent=1.6526, interpMode=RCIM_Cubic, leaveTangent=1.6526, tangentMode=RCTM_User, time=1, value=1}
        ]}}})
    Module.Param0WriteEnabled=true
    Module.Param1WriteEnabled=false
    Module.Param2WriteEnabled=true
    Module.Param3WriteEnabled=true
   )
  ]
  rendererClasses [&NiagaraSpriteRendererProperties]
  renderers [{rendererClass=&NiagaraSpriteRendererProperties}]
 }
}
summary {
 systemName=NS_Emote_Omega_InstrumentBurstUp
 emitters [
  {emitterName=Splash_Foam, rendererClasses [&NiagaraSpriteRendererProperties]}
  {emitterName=Glow001, rendererClasses []}
  {emitterName=Impact_Streaks001, rendererClasses []}
  {emitterName=Droplets001, rendererClasses []}
  {emitterName=EnergyWave_Burst, rendererClasses [&NiagaraMeshRendererProperties]}
 ]
 userVariables [
  User.bIsFrontEnd {type=NiagaraBool}
  User.bIsFrontEndPreview {type=NiagaraBool}
  User.ParaGlideLeanAlpha {type=NiagaraFloat, defaultValue=1, description=""}
 ]
}
