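# Niagara slow-motion trail effect for the Omega emote, attaching glow sprites and colour-shifting ribbons to skeletal mesh bone or socket points.
# One glow and ribbon pair always plays, while a duplicate pair on the second point is gated by the float User.bDrums.
# Burst glows track their point each update as ribbons spawn along its motion, then taper off.
# Ribbons widen and fade with time-of-day brightness scaling and curve-driven dynamic material parameters.
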NiagaraSystem /BRCosmetics/Animation/Game/MainPlayer/Emotes/Omega/FX/NS_Emote_Omega_SlomoTrail
dependencies {
 usedDataInterfaces [
  &NiagaraDataInterfaceArrayColor
  &NiagaraDataInterfaceCurve
  &NiagaraDataInterfaceColorCurve
  &NiagaraDataInterfaceSkeletalMesh
 ]
 usedDynamicInputs []
 usedModules [
  &SystemState
  &self:SystemSpawnScript.SystemScriptSource.SystemScriptGraph.NiagaraNodeAssignment_0.SetVariables_#1
 ]
}
emitters {
 Glow {
  emitterUpdateScript [
   &EmitterState()
   &SpawnBurst_Instantaneous(Use Spawn Probability=false, UseLoopCountLimit=false)
  ]
  particleSpawnScript [&InitializeParticle(
    Color=&SelectLinearColorFromArray(
     Array Sampling Mode=::NewEnumerator1
     Direct Array Index=0
     Color Selection Array=DataInterface() {propertyValues {ColorData [<rgba 0 0 0 1>, <rgba 1 1 1 1>]}}
     Random Seed=?
    )
    Lifetime=2.15
    Sprite Size Mode=::NewEnumerator3
    Uniform Sprite Size=45
    UsePositionOffset=false
    Write Material Random=true
   )]
  particleUpdateScript [
   &ParticleState()
   &SkeletalMeshLocation(MaskSpawnGroup=false, Mesh Sampling Type=::NewEnumerator5, Skeletal Mesh=$System.Skeletal Mesh, Socket Sampling Mode=::NewEnumerator5)
   &ScaleSpriteSize(Uniform Curve Sprite Scale=DataInterface() {propertyValues {Curve {keys [{tangentMode=RCTM_User, time=0, value=0}, <at 0.05 1.2>, {tangentMode=RCTM_User, time=0.125, value=1}]}}})
   &ScaleColor(
    Linear Color Curve=DataInterface() {propertyValues {
      AlphaCurve {keys [<at 0 0 RCIM_Cubic>, {arriveTangent=2, interpMode=RCIM_Cubic, leaveTangent=2, time=0.4, value=1}, <at 0.5 1>, <at 0.75 0>]}
      BlueCurve {keys [<at 0 0 RCIM_Cubic>, {interpMode=RCIM_Cubic, tangentMode=RCTM_User, time=0.2852, value=0}, <at 0.5204 0>, <at 0.7867 0.1833>, <at 1 1>]}
      GreenCurve {keys [
        <at 0 0.4167 RCIM_Cubic>
        {interpMode=RCIM_Cubic, tangentMode=RCTM_User, time=0.2852, value=0.15}
        <at 0.5204 0.0667>
        <at 0.7867 0.02>
        <at 1 0.05>
       ]}
      RedCurve {keys [<at 0 5 RCIM_Cubic>, {interpMode=RCIM_Cubic, tangentMode=RCTM_User, time=0.2852, value=3}, <at 0.5204 2>, <at 0.7867 1>, <at 1 0.6833>]}
     }}
    Scale Mode=::NewEnumerator2
   )
   /SetVariables(
    Particles.GlowColor=&ColorFromCurve(ColorCurve=DataInterface() {propertyValues {
       AlphaCurve {keys [<at 0 0 RCIM_Cubic>, {arriveTangent=0.6, interpMode=RCIM_Cubic, leaveTangent=0.6, time=0.4, value=0.3}, <at 0.5 0.3>, <at 0.75 0>]}
       BlueCurve {keys [<at 0 0 RCIM_Cubic>, {interpMode=RCIM_Cubic, tangentMode=RCTM_User, time=0.2852, value=0}, <at 0.5204 0>, <at 0.7867 0.1833>, <at 1 1>]}
       GreenCurve {keys [
         <at 0 0.4167 RCIM_Cubic>
         {interpMode=RCIM_Cubic, tangentMode=RCTM_User, time=0.2852, value=0.15}
         <at 0.5204 0.0667>
         <at 0.7867 0.02>
         <at 1 0.05>
        ]}
       RedCurve {keys [<at 0 5 RCIM_Cubic>, {interpMode=RCIM_Cubic, tangentMode=RCTM_User, time=0.2852, value=3}, <at 0.5204 2>, <at 0.7867 1>, <at 1 0.6833>]}
      }})
    Particles.GlowSize=&Multiply_Vector2D_ByFloat(Float=55, Vector2D=<1 1>)
   ) {isSetParameters=true, module=SetVariables_#2}
  ]
  rendererClasses [&NiagaraSpriteRendererProperties]
  renderers [
   {rendererClass=&NiagaraSpriteRendererProperties}
   {rendererClass=&NiagaraSpriteRendererProperties, rendererIndex=1}
  ]
 }
 Glow001 {
  emitterUpdateScript [
   &EmitterState()
   &SpawnBurst_Instantaneous(
    Spawn Count=&MakeIntFromFloat(Float=&MakeCustomFloatFromBool(
      Bool=&SetBoolByFloatComparison(A=$User.bDrums, B=0, Comparison Type=::NewEnumerator4)
      True Float=1
      False Float=0
     ))
    Use Spawn Probability=false
    UseLoopCountLimit=false
   )
  ]
  particleSpawnScript [&InitializeParticle(
    Color=&SelectLinearColorFromArray(
     Array Sampling Mode=::NewEnumerator1
     Direct Array Index=0
     Color Selection Array=DataInterface() {propertyValues {ColorData [<rgba 0 0 0 1>, <rgba 1 1 1 1>]}}
     Random Seed=?
    )
    Lifetime=2.15
    Sprite Size Mode=::NewEnumerator3
    Uniform Sprite Size=45
    UsePositionOffset=false
    Write Material Random=true
   )]
  particleUpdateScript [
   &ParticleState()
   &SkeletalMeshLocation(
    Bone / Socket Index=1
    MaskSpawnGroup=false
    Mesh Sampling Type=::NewEnumerator5
    Skeletal Mesh=$System.Skeletal Mesh
    Socket Sampling Mode=::NewEnumerator5
   )
   &ScaleSpriteSize(Uniform Curve Sprite Scale=DataInterface() {propertyValues {Curve {keys [{tangentMode=RCTM_User, time=0, value=0}, <at 0.05 1.2>, {tangentMode=RCTM_User, time=0.125, value=1}]}}})
   &ScaleColor(
    Linear Color Curve=DataInterface() {propertyValues {
      AlphaCurve {keys [<at 0 0 RCIM_Cubic>, {arriveTangent=2, interpMode=RCIM_Cubic, leaveTangent=2, time=0.4, value=1}, <at 0.5 1>, <at 0.75 0>]}
      BlueCurve {keys [<at 0 0 RCIM_Cubic>, {interpMode=RCIM_Cubic, tangentMode=RCTM_User, time=0.2852, value=0}, <at 0.5204 0>, <at 0.7867 0.1833>, <at 1 1>]}
      GreenCurve {keys [
        <at 0 0.4167 RCIM_Cubic>
        {interpMode=RCIM_Cubic, tangentMode=RCTM_User, time=0.2852, value=0.15}
        <at 0.5204 0.0667>
        <at 0.7867 0.02>
        <at 1 0.05>
       ]}
      RedCurve {keys [<at 0 5 RCIM_Cubic>, {interpMode=RCIM_Cubic, tangentMode=RCTM_User, time=0.2852, value=3}, <at 0.5204 2>, <at 0.7867 1>, <at 1 0.6833>]}
     }}
    Scale Mode=::NewEnumerator2
   )
   /SetVariables(
    Particles.GlowColor=&ColorFromCurve(ColorCurve=DataInterface() {propertyValues {
       AlphaCurve {keys [<at 0 0 RCIM_Cubic>, {arriveTangent=0.6, interpMode=RCIM_Cubic, leaveTangent=0.6, time=0.4, value=0.3}, <at 0.5 0.3>, <at 0.75 0>]}
       BlueCurve {keys [<at 0 0 RCIM_Cubic>, {interpMode=RCIM_Cubic, tangentMode=RCTM_User, time=0.2852, value=0}, <at 0.5204 0>, <at 0.7867 0.1833>, <at 1 1>]}
       GreenCurve {keys [
         <at 0 0.4167 RCIM_Cubic>
         {interpMode=RCIM_Cubic, tangentMode=RCTM_User, time=0.2852, value=0.15}
         <at 0.5204 0.0667>
         <at 0.7867 0.02>
         <at 1 0.05>
        ]}
       RedCurve {keys [<at 0 5 RCIM_Cubic>, {interpMode=RCIM_Cubic, tangentMode=RCTM_User, time=0.2852, value=3}, <at 0.5204 2>, <at 0.7867 1>, <at 1 0.6833>]}
      }})
    Particles.GlowSize=&Multiply_Vector2D_ByFloat(Float=55, Vector2D=<1 1>)
   ) {isSetParameters=true, module=SetVariables_#2}
  ]
  rendererClasses [&NiagaraSpriteRendererProperties]
  renderers [
   {rendererClass=&NiagaraSpriteRendererProperties}
   {rendererClass=&NiagaraSpriteRendererProperties, rendererIndex=1}
  ]
 }
 Ribbon_AllInstruments {
  emitterUpdateScript [
   &EmitterState()
   &SpawnRate(
    SpawnRate=&FloatFromCurve(FloatCurve=DataInterface() {propertyValues {Curve {keys [<at 0 1>,<at 1.29 1>,<at 1.63 0>]}}}, CurveIndex=$Emitter.Age, Scale Curve=15)
    Use Spawn Probability=false
   )
  ]
  particleSpawnScript [
   &InitializeRibbon(
    Lifetime=1.2
    Module.Write Color=true
    Module.Write Lifetime=true
    Module.Write Mass=false
    Module.Write Material Random=true
    Module.Write Position=false
    Module.Write RibbonTwist=false
    Module.Write RibbonWidth=true
    Ribbon Width=20
   )
   &SkeletalMeshLocation(MaskSpawnGroup=false, Mesh Sampling Type=::NewEnumerator5, Skeletal Mesh=$System.Skeletal Mesh, Socket Sampling Mode=::NewEnumerator5)
  ]
  particleUpdateScript [
   &ParticleState()
   &SolveForcesAndVelocity(Clamp Velocity=false, Limit Acceleration=false)
   &Color(
    Color=&ColorFromCurve(ColorCurve=DataInterface() {propertyValues {
       AlphaCurve {keys [<at 0 1 RCIM_Cubic>,<at 0.5 1>,<at 0.9949 0>]}
       BlueCurve {keys [
         {arriveTangent=4.0072, interpMode=RCIM_Cubic, time=0, value=0}
         <at 0.1526 0>
         <at 0.3189 0>
         <at 0.5036 1>
         <at 0.6607 1>
         <at 1 1 RCIM_Cubic>
        ]}
       GreenCurve {keys [
         {arriveTangent=-0.4007, interpMode=RCIM_Cubic, time=0, value=0.0833}
         <at 0.1526 0.0333>
         <at 0.3189 0.0333>
         <at 0.5036 0>
         <at 0.6607 0.2>
         <at 1 0.4 RCIM_Cubic>
        ]}
       RedCurve {keys [
         {arriveTangent=-11.8213, interpMode=RCIM_Cubic, time=0, value=1}
         <at 0.1526 1>
         <at 0.3189 1>
         <at 0.5036 0.6667>
         <at 0.6607 0.2>
         <at 1 0.1 RCIM_Cubic>
        ]}
      }})
    Scale Color=&Multiply_Vector(
     A=&VectorFromFloat(Value=&FNTimeOfDayScalar(Night Scale=0.5, Morning Scale=0.25, Evening Scale=0.5))
     B=&FN_Select_Vector_from_Time_Of_Day(Morning=<1 1 1>, Day=<1 1 1>, Evening=<1 1 1>, Night=<0.25 1 1>)
    )
   )
   &ScaleColor(
    Color Value To Scale=$Particles.Color
    Scale RGB=&VectorFromFloat(Value=&FloatFromCurve(
      FloatCurve=DataInterface() {propertyValues {Curve {keys [
          {interpMode=RCIM_Cubic, leaveTangent=-5.4692, tangentMode=RCTM_Break, time=0, value=2}
          {arriveTangent=0.1008, interpMode=RCIM_Cubic, leaveTangent=0.1008, tangentMode=RCTM_User, time=0.3, value=1}
          {interpMode=RCIM_Cubic, tangentMode=RCTM_User, time=1, value=0}
         ]}}}
      Scale Curve=&FNTimeOfDayScalar(Night Scale=6, Morning Scale=4, Day Scale=3, Evening Scale=4)
     ))
    ScaleA=true
    ScaleRGB=true
   )
   &ScaleRibbonWidth(Ribbon Width Scale=&FloatFromCurve(FloatCurve=DataInterface() {propertyValues {Curve {keys [<at 0 1 RCIM_Cubic>, {arriveTangent=6.3317, interpMode=RCIM_Cubic, leaveTangent=6.3317, tangentMode=RCTM_User, time=1, value=3}]}}})) {module=ScaleRibbonWidth001}
   &DynamicMaterialParameters(
    Index 0 Param 1=&FloatFromCurve(FloatCurve=DataInterface() {propertyValues {Curve {keys [
         {arriveTangent=0.2015, interpMode=RCIM_Cubic, leaveTangent=0.2015, tangentMode=RCTM_User, time=0, value=1}
         {arriveTangent=-1.5941, interpMode=RCIM_Cubic, leaveTangent=-1.5941, tangentMode=RCTM_User, time=0.3, value=0.5}
         {interpMode=RCIM_Cubic, tangentMode=RCTM_User, time=1, value=0}
        ]}}})
    Index 0 Param 2=6
    Index 0 Param 3=-0.1
    Index 0 Param 4=&FloatFromCurve(
     FloatCurve=DataInterface() {propertyValues {Curve {keys [
         <at 0 0 RCIM_Cubic>
         {interpMode=RCIM_Cubic, tangentMode=RCTM_User, time=0.25, value=1}
         {arriveTangent=-0.6219, interpMode=RCIM_Cubic, leaveTangent=-0.6219, tangentMode=RCTM_User, time=1, value=0}
        ]}}}
     CurveIndex=$Emitter.NormalizedLoopAge
    )
    Module.Index0Float0Write=true
    Module.Index0Float1Write=true
    Module.Index0Float2Write=true
    Module.Index0Float3Write=true
   )
  ]
  rendererClasses [&NiagaraRibbonRendererProperties]
  renderers [{rendererClass=&NiagaraRibbonRendererProperties}]
 }
 Ribbon_Drums {
  emitterUpdateScript [
   &EmitterState()
   &SpawnRate(
    SpawnRate=&FloatFromCurve(
     FloatCurve=DataInterface() {propertyValues {Curve {keys [<at 0 1>,<at 1.29 1>,<at 1.63 0>]}}}
     CurveIndex=$Emitter.Age
     Scale Curve=&MakeCustomFloatFromBool(
      Bool=&SetBoolByFloatComparison(A=$User.bDrums, B=0, Comparison Type=::NewEnumerator4)
      True Float=15
      False Float=0
     )
    )
    Use Spawn Probability=false
   )
  ]
  particleSpawnScript [
   &InitializeRibbon(
    Lifetime=1.2
    Module.Write Color=true
    Module.Write Lifetime=true
    Module.Write Mass=false
    Module.Write Material Random=true
    Module.Write Position=false
    Module.Write RibbonTwist=false
    Module.Write RibbonWidth=true
    Ribbon Width=20
   )
   &SkeletalMeshLocation(
    Bone / Socket Index=1
    MaskSpawnGroup=false
    Mesh Sampling Type=::NewEnumerator5
    Skeletal Mesh=$System.Skeletal Mesh
    Socket Sampling Mode=::NewEnumerator5
   )
  ]
  particleUpdateScript [
   &ParticleState()
   &SolveForcesAndVelocity(Clamp Velocity=false, Limit Acceleration=false)
   &Color(
    Color=&ColorFromCurve(ColorCurve=DataInterface() {propertyValues {
       AlphaCurve {keys [<at 0 1 RCIM_Cubic>,<at 0.5 1>,<at 0.9949 0>]}
       BlueCurve {keys [
         {arriveTangent=4.0072, interpMode=RCIM_Cubic, time=0, value=0}
         <at 0.1526 0>
         <at 0.3189 0>
         <at 0.5036 1>
         <at 0.6607 1>
         <at 1 1 RCIM_Cubic>
        ]}
       GreenCurve {keys [
         {arriveTangent=-0.4007, interpMode=RCIM_Cubic, time=0, value=0.0833}
         <at 0.1526 0.0333>
         <at 0.3189 0.0333>
         <at 0.5036 0>
         <at 0.6607 0.2>
         <at 1 0.4 RCIM_Cubic>
        ]}
       RedCurve {keys [
         {arriveTangent=-11.8213, interpMode=RCIM_Cubic, time=0, value=1}
         <at 0.1526 1>
         <at 0.3189 1>
         <at 0.5036 0.6667>
         <at 0.6607 0.2>
         <at 1 0.1 RCIM_Cubic>
        ]}
      }})
    Scale Color=&Multiply_Vector(
     A=&VectorFromFloat(Value=&FNTimeOfDayScalar(Night Scale=0.5, Morning Scale=0.25, Evening Scale=0.5))
     B=&FN_Select_Vector_from_Time_Of_Day(Morning=<1 1 1>, Day=<1 1 1>, Evening=<1 1 1>, Night=<0.25 1 1>)
    )
   )
   &ScaleColor(
    Color Value To Scale=$Particles.Color
    Scale RGB=&VectorFromFloat(Value=&FloatFromCurve(
      FloatCurve=DataInterface() {propertyValues {Curve {keys [
          {interpMode=RCIM_Cubic, leaveTangent=-5.4692, tangentMode=RCTM_Break, time=0, value=2}
          {arriveTangent=0.1008, interpMode=RCIM_Cubic, leaveTangent=0.1008, tangentMode=RCTM_User, time=0.3, value=1}
          {interpMode=RCIM_Cubic, tangentMode=RCTM_User, time=1, value=0}
         ]}}}
      Scale Curve=&FNTimeOfDayScalar(Night Scale=6, Morning Scale=4, Day Scale=3, Evening Scale=4)
     ))
    ScaleA=true
    ScaleRGB=true
   )
   &ScaleRibbonWidth(Ribbon Width Scale=&FloatFromCurve(FloatCurve=DataInterface() {propertyValues {Curve {keys [<at 0 1 RCIM_Cubic>, {arriveTangent=6.3317, interpMode=RCIM_Cubic, leaveTangent=6.3317, tangentMode=RCTM_User, time=1, value=3}]}}})) {module=ScaleRibbonWidth001}
   &DynamicMaterialParameters(
    Index 0 Param 1=&FloatFromCurve(FloatCurve=DataInterface() {propertyValues {Curve {keys [
         {arriveTangent=0.2015, interpMode=RCIM_Cubic, leaveTangent=0.2015, tangentMode=RCTM_User, time=0, value=1}
         {arriveTangent=-1.5941, interpMode=RCIM_Cubic, leaveTangent=-1.5941, tangentMode=RCTM_User, time=0.3, value=0.5}
         {interpMode=RCIM_Cubic, tangentMode=RCTM_User, time=1, value=0}
        ]}}})
    Index 0 Param 2=6
    Index 0 Param 3=-0.1
    Index 0 Param 4=&FloatFromCurve(
     FloatCurve=DataInterface() {propertyValues {Curve {keys [
         <at 0 0 RCIM_Cubic>
         {interpMode=RCIM_Cubic, tangentMode=RCTM_User, time=0.25, value=1}
         {arriveTangent=-0.6219, interpMode=RCIM_Cubic, leaveTangent=-0.6219, tangentMode=RCTM_User, time=1, value=0}
        ]}}}
     CurveIndex=$Emitter.NormalizedLoopAge
    )
    Module.Index0Float0Write=true
    Module.Index0Float1Write=true
    Module.Index0Float2Write=true
    Module.Index0Float3Write=true
   )
  ]
  rendererClasses [&NiagaraRibbonRendererProperties]
  renderers [{rendererClass=&NiagaraRibbonRendererProperties}]
 }
}
summary {
 systemName=NS_Emote_Omega_SlomoTrail
 emitters [
  {emitterName=Glow, rendererClasses [&NiagaraSpriteRendererProperties]}
  {emitterName=Ribbon_AllInstruments, rendererClasses [&NiagaraRibbonRendererProperties]}
  {emitterName=Glow001, rendererClasses [&NiagaraSpriteRendererProperties]}
  {emitterName=Ribbon_Drums, rendererClasses [&NiagaraRibbonRendererProperties]}
 ]
 userVariables [User.bDrums {type=NiagaraFloat, defaultValue=0, description=""}]
}
